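; Axis_Align_Switch is a PCG subgraph that routes To Min, To Center or To Max param data to one output, based on justify-from and justify-to enums.
; A comparison on Justify_to chooses whether it or Justify_from is the effective target.
; Three comparisons on that target each combine with a Justify_from versus "which" check.
; Each result gates a Select with only input B wired, all feeding Out.
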(graph Axis_Align_Switch
  (params (Justify_to :type (Enum /PCGPrimitives/Primitives/Transform/Subgraphs/EPCGJustifyTo.EPCGJustifyTo)) (Justify_from :type (Enum /PCGPrimitives/Primitives/Transform/Subgraphs/EPCGJustifyFrom.EPCGJustifyFrom)) (which :type (Enum /PCGPrimitives/Primitives/Transform/Subgraphs/EPCGJustifyTo.EPCGJustifyTo)))
  (inputs (To%20Min%20Out :type Param) (To%20Center%20Out :type Param) (To%20Max%20Out :type Param))
  (bind AttributeCompareOp_0 (Attribute%20Compare%20Op :InA Justify_from :InB which))
  (bind AttributeCompareOp_1 (Attribute%20Compare%20Op :InA Justify_to))
  (bind Select_0 (Select :Input%20A Justify_to :Input%20B Justify_from))
  (bind AttributeCompareOp_2 (Attribute%20Compare%20Op))
  (bind AttributeBooleanOp_0 (Attribute%20Boolean%20Op))
  (bind Select_1 (Select))
  (bind AttributeCompareOp_3 (Attribute%20Compare%20Op))
  (bind AttributeBooleanOp_1 (Attribute%20Boolean%20Op))
  (bind AttributeCompareOp_4 (Attribute%20Compare%20Op))
  (bind AttributeBooleanOp_2 (Attribute%20Boolean%20Op))
  (bind Select_2 (Select))
  (bind Select_3 (Select))
  (reroute Min%20to%20Center To%20Center%20Out)
  (reroute Min%20to%20MAx To%20Max%20Out)
  (reroute Min%20to%20Min To%20Min%20Out)
  (connect AttributeCompareOp_0 Out AttributeBooleanOp_0 InA)
  (connect AttributeCompareOp_2 Out AttributeBooleanOp_0 InB)
  (connect AttributeCompareOp_0 Out AttributeBooleanOp_1 InA)
  (connect AttributeCompareOp_3 Out AttributeBooleanOp_1 InB)
  (connect AttributeCompareOp_0 Out AttributeBooleanOp_2 InA)
  (connect AttributeCompareOp_4 Out AttributeBooleanOp_2 InB)
  (connect Select_0 Out AttributeCompareOp_2 InA)
  (connect Select_0 Out AttributeCompareOp_3 InA)
  (connect Select_0 Out AttributeCompareOp_4 InA)
  (connect AttributeCompareOp_1 Out Select_0 bUseInputB)
  (connect AttributeBooleanOp_0 Out Select_1 bUseInputB)
  (connect Min%20to%20Min Out Select_1 Input%20B)
  (connect AttributeBooleanOp_1 Out Select_2 bUseInputB)
  (connect Min%20to%20Center Out Select_2 Input%20B)
  (connect AttributeBooleanOp_2 Out Select_3 bUseInputB)
  (connect Min%20to%20MAx Out Select_3 Input%20B)
  (output Out Select_1)
  (output Out Select_2)
  (output Out Select_3))
