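; PCG helper subgraph in the PCGPrimitives transform set that routes nine attribute-set inputs, one per Min/Center/Max from-to justification pairing, toward a single output.
; Inputs are grouped by source justification into three Axis_Align_Switch instances, each taking To Min, To Center and To Max.
; All three instances receive the same Justify_X_from and Justify_X_to enums, and their results merge on Out.
; The Min-to-Min input pin is named just "Out".
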(graph Axis_Align_Switch_TopGraph
  (params (Justify_X_to :type (Enum /PCGPrimitives/Primitives/Transform/Subgraphs/EPCGJustifyTo.EPCGJustifyTo)) (Justify_X_from :type (Enum /PCGPrimitives/Primitives/Transform/Subgraphs/EPCGJustifyFrom.EPCGJustifyFrom)))
  (inputs (Out :type Param) (Min%20To%20Center%20Out :type Param) (Min%20To%20MAx%20Out :type Param) (Center%20To%20Min%20Out :type Param) (Center%20to%20Center%20Out :type Param) (Center%20To%20Max%20Out :type Param) (Max%20To%20Min%20Out :type Param) (Max%20To%20Center%20Out :type Param) (Max%20To%20Max%20Out :type Param))
  (bind Subgraph_0 (Subgraph :Graph /PCGPrimitives/Primitives/Transform/Subgraphs/Axis_Align_Switch.Axis_Align_Switch :Justify_from Justify_X_from :Justify_to Justify_X_to))
  (bind Subgraph_1 (Subgraph :Graph /PCGPrimitives/Primitives/Transform/Subgraphs/Axis_Align_Switch.Axis_Align_Switch :Justify_from Justify_X_from :Justify_to Justify_X_to))
  (bind Subgraph_2 (Subgraph :Graph /PCGPrimitives/Primitives/Transform/Subgraphs/Axis_Align_Switch.Axis_Align_Switch :Justify_from Justify_X_from :Justify_to Justify_X_to))
  (reroute Center%20To%20Center Center%20to%20Center%20Out)
  (reroute Center%20to%20Max Center%20To%20Max%20Out)
  (reroute Center%20to%20Min Center%20To%20Min%20Out)
  (reroute Max%20to%20Center Max%20To%20Center%20Out)
  (reroute Max%20to%20Max Max%20To%20Max%20Out)
  (reroute Max%20to%20Min Max%20To%20Min%20Out)
  (reroute Min%20to%20Center Min%20To%20Center%20Out)
  (reroute Min%20to%20MAx Min%20To%20MAx%20Out)
  (reroute Min%20to%20Min Out)
  (connect Min%20to%20Center Out Subgraph_0 To%20Center%20Out)
  (connect Min%20to%20MAx Out Subgraph_0 To%20Max%20Out)
  (connect Min%20to%20Min Out Subgraph_0 To%20Min%20Out)
  (connect Center%20To%20Center Out Subgraph_1 To%20Center%20Out)
  (connect Center%20to%20Max Out Subgraph_1 To%20Max%20Out)
  (connect Center%20to%20Min Out Subgraph_1 To%20Min%20Out)
  (connect Max%20to%20Center Out Subgraph_2 To%20Center%20Out)
  (connect Max%20to%20Max Out Subgraph_2 To%20Max%20Out)
  (connect Max%20to%20Min Out Subgraph_2 To%20Min%20Out)
  (output Out Subgraph_0)
  (output Out Subgraph_1)
  (output Out Subgraph_2))
