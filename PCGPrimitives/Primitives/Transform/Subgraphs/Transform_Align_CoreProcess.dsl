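; PCG core subgraph that aligns primary data (tag Input:0) to the remaining inputs by translating points, splines and polygons per axis.
; Each set's combined bounds yield min, center and max, from which nine candidate offsets are computed.
; An Axis_Align_Switch_TopGraph per axis picks an offset from the Justify from/to enums, forming one translation.
; With UseParentAxis, alignment runs in the parent actor's local space and results are transformed back.
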(graph Transform_Align_CoreProcess
  (params (Justify_X_from :type (Enum /PCGPrimitives/Primitives/Transform/Subgraphs/EPCGJustifyFrom.EPCGJustifyFrom)) (Justify_X_to :type (Enum /PCGPrimitives/Primitives/Transform/Subgraphs/EPCGJustifyTo.EPCGJustifyTo)) (Justify_Y_from :type (Enum /PCGPrimitives/Primitives/Transform/Subgraphs/EPCGJustifyFrom.EPCGJustifyFrom)) (Justify_Y_to :type (Enum /PCGPrimitives/Primitives/Transform/Subgraphs/EPCGJustifyTo.EPCGJustifyTo)) (Justify_Z_from :type (Enum /PCGPrimitives/Primitives/Transform/Subgraphs/EPCGJustifyFrom.EPCGJustifyFrom)) (Justify_Z_to :type (Enum /PCGPrimitives/Primitives/Transform/Subgraphs/EPCGJustifyTo.EPCGJustifyTo)) (UseParentAxis :type Bool :default true))
  (bind FilterDataByType_1 (Filter%20Data%20By%20Type :TargetType (struct :Ids [(struct :Struct /Script/PCG.PCGDataTypeInfoPoint)])))
  (bind FilterDataByType_2 (Filter%20Data%20By%20Type :TargetType (struct :Ids [(struct :Struct /Script/PCG.PCGDataTypeInfoSpline)])))
  (bind FilterDataByType_3 (Filter%20Data%20By%20Type :TargetType (struct :Ids [(struct :Struct /Script/PCG.PCGDataTypeInfoPolygon2D)])))
  (bind AttributeTransformOp_16 (Attribute%20Transform%20Op :InputSource1 $Transform :Operation Invert))
  (bind FilterDataByType_5 (Filter%20Data%20By%20Type :TargetType (struct :Ids [(struct :Struct /Script/PCG.PCGDataTypeInfoPoint)])))
  (bind FilterDataByType_6 (Filter%20Data%20By%20Type :TargetType (struct :Ids [(struct :Struct /Script/PCG.PCGDataTypeInfoSpline)])))
  (bind GetSplineControlPoints_4 (Get%20Spline%20Control%20Points))
  (bind FilterDataByType_7 (Filter%20Data%20By%20Type :TargetType (struct :Ids [(struct :Struct /Script/PCG.PCGDataTypeInfoPolygon2D)])))
  (bind ToPoint_0 (To%20Point))
  (bind AttributeTransformOp_15 (Attribute%20Transform%20Op :InputSource1 $transform :InputSource2 $transform))
  (bind Select_8 (Select :bUseInputB-param UseParentAxis))
  ;[0 0 0 0.5 12] Primary/Secondary Inputs
  (bind FilterDataByTag_1 (Filter%20Data%20By%20Tag :SelectedTags "Input:0"))
  ;[0.15 0.15 0.15 0.5 70] PARSING USER inputs
  (bind CreateConstant_4 (Create%20Constant :AttributeTypes (struct :Type Vector :VectorValue (1.0 1.0 1.0))))
  (bind AttributeMathsOp_8 (Attribute%20Maths%20Op :Operation Clamp :In Justify_X_from))
  (bind CreateConstant_5 (Create%20Constant :AttributeTypes (struct :Type Float)))
  (bind Subgraph_3 (Subgraph :Graph /PCGPrimitives/Primitives/Transform/Subgraphs/Axis_Align_Switch_TopGraph.Axis_Align_Switch_TopGraph :Justify_X_from Justify_X_from :Justify_X_to Justify_X_to))
  (bind ExtractAttribute_0 (Extract%20Attribute%20at%20Index :bForceOutputAttributeToBeInElementsDomain true :InputSource Align.X :OutputAttributeName Value))
  (bind Select_4 (Select))
  (bind AttributeMathsOp_14 (Attribute%20Maths%20Op :Operation Clamp :In Justify_Y_from))
  (bind CreateConstant_6 (Create%20Constant :AttributeTypes (struct :Type Float)))
  (bind Subgraph_4 (Subgraph :Graph /PCGPrimitives/Primitives/Transform/Subgraphs/Axis_Align_Switch_TopGraph.Axis_Align_Switch_TopGraph :Justify_X_from Justify_Y_from :Justify_X_to Justify_Y_to))
  (bind ExtractAttribute_1 (Extract%20Attribute%20at%20Index :bForceOutputAttributeToBeInElementsDomain true :InputSource Align.Y :OutputAttributeName Value))
  (bind Select_5 (Select))
  (bind AttributeMathsOp_15 (Attribute%20Maths%20Op :Operation Clamp :In Justify_Z_from))
  (bind CreateConstant_7 (Create%20Constant :AttributeTypes (struct :Type Float)))
  (bind Subgraph_5 (Subgraph :Graph /PCGPrimitives/Primitives/Transform/Subgraphs/Axis_Align_Switch_TopGraph.Axis_Align_Switch_TopGraph :Justify_X_from Justify_Z_from :Justify_X_to Justify_Z_to))
  (bind ExtractAttribute_2 (Extract%20Attribute%20at%20Index :bForceOutputAttributeToBeInElementsDomain true :InputSource Align.Z :OutputAttributeName Value))
  (bind Select_6 (Select))
  (bind MakeVectorAttribute_2 (Make%20Vector%20Attribute :OutputType Vector))
  (bind MakeTransformAttribute_1 (Make%20Transform%20Attribute))
  ;[0.15 0.15 0.15 0.5 70] COMPARE POSITION for each axis
  (bind MergePoints_1 (Merge%20Points))
  (bind MergePoints_2 (Merge%20Points))
  (bind CombinePoints_0 (Combine%20Points :bUseFirstPointTransform false))
  (bind CombinePoints_1 (Combine%20Points :bUseFirstPointTransform false))
  (bind ExtractAttribute_3 (Extract%20Attribute%20at%20Index :bForceOutputAttributeToBeInElementsDomain true :InputSource $position))
  (bind AttributeRename_1 (Attribute%20Rename :AttributeToRename position :NewAttributeName Center))
  (bind ExtractAttribute_8 (Extract%20Attribute%20at%20Index :bForceOutputAttributeToBeInElementsDomain true :InputSource $BoundsMax))
  (bind AttributeMathsOp_23 (Attribute%20Maths%20Op))
  (bind ExtractAttribute_7 (Extract%20Attribute%20at%20Index :bForceOutputAttributeToBeInElementsDomain true :InputSource $BoundsMin))
  (bind AttributeMathsOp_24 (Attribute%20Maths%20Op))
  (bind MergeAttributes_3 (Merge%20Attributes))
  (bind AttributeFilter_0 (Filter%20Attribute%20Elements :AttributeTypes (struct :Type Integer32) :bUseConstantThreshold true :Operator Equal :TargetAttribute $Index :ThresholdAttribute @Last))
  (bind ExtractAttribute_9 (Extract%20Attribute%20at%20Index :bForceOutputAttributeToBeInElementsDomain true :InputSource $position))
  (bind AttributeRename_0 (Attribute%20Rename :AttributeToRename position :NewAttributeName Center))
  (bind ExtractAttribute_11 (Extract%20Attribute%20at%20Index :bForceOutputAttributeToBeInElementsDomain true :InputSource $BoundsMax))
  (bind AttributeMathsOp_21 (Attribute%20Maths%20Op))
  (bind ExtractAttribute_10 (Extract%20Attribute%20at%20Index :bForceOutputAttributeToBeInElementsDomain true :InputSource $BoundsMin))
  (bind AttributeMathsOp_22 (Attribute%20Maths%20Op))
  (bind MergeAttributes_4 (Merge%20Attributes))
  (bind AttributeFilter_1 (Filter%20Attribute%20Elements :AttributeTypes (struct :Type Integer32) :bUseConstantThreshold true :Operator Equal :TargetAttribute $Index :ThresholdAttribute @Last))
  (bind AttributeMathsOp_7 (Attribute%20Maths%20Op :InputSource1 BoundsMax :InputSource2 BoundsMin :Operation Subtract :OutputTarget Align))
  (bind AttributeMathsOp_9 (Attribute%20Maths%20Op :InputSource1 BoundsMin :InputSource2 BoundsMax :Operation Subtract :OutputTarget Align))
  (bind AttributeMathsOp_10 (Attribute%20Maths%20Op :InputSource1 BoundsMin :InputSource2 Center :Operation Subtract :OutputTarget Align))
  (bind AttributeMathsOp_11 (Attribute%20Maths%20Op :InputSource1 BoundsMax :InputSource2 Center :Operation Subtract :OutputTarget Align))
  (bind AttributeMathsOp_12 (Attribute%20Maths%20Op :InputSource1 Center :InputSource2 BoundsMin :Operation Subtract :OutputTarget Align))
  (bind AttributeMathsOp_13 (Attribute%20Maths%20Op :InputSource1 Center :InputSource2 BoundsMax :Operation Subtract :OutputTarget Align))
  (bind AttributeMathsOp_17 (Attribute%20Maths%20Op :InputSource1 BoundsMin :InputSource2 BoundsMin :Operation Subtract :OutputTarget Align))
  (bind AttributeMathsOp_18 (Attribute%20Maths%20Op :InputSource1 Center :InputSource2 Center :Operation Subtract :OutputTarget Align))
  (bind AttributeMathsOp_19 (Attribute%20Maths%20Op :InputSource1 BoundsMax :InputSource2 BoundsMax :Operation Subtract :OutputTarget Align))
  ;[0.05861 0.15 0.1147 0.5 30] POLYGON
  (bind AttributeTransformOp_14 (Attribute%20Transform%20Op :InputSource1 $Transform :Operation Invert))
  (bind ToPoint_1 (To%20Point))
  (bind AttributeTransformOp_13 (Attribute%20Transform%20Op :InputSource1 $transform :InputSource2 $transform))
  (bind Select_3 (Select :bUseInputB-param UseParentAxis))
  (bind AttributeVectorOp_2 (Attribute%20Vector%20Op :InputSource1 $Position :Operation TransformLocation))
  (bind AttributeTransformOp_8 (Attribute%20Transform%20Op :InputSource1 $transform :InputSource2 $transform))
  (bind Select_12 (Select :bUseInputB-param UseParentAxis))
  (bind CreatePolygon2D_1 (Create%20Polygon%202D))
  ;[0.09368 0.239 0.6198 0.5 30] POINTS
  (bind AttributeTransformOp_3 (Attribute%20Transform%20Op :InputSource1 $Transform :Operation Invert))
  (bind AttributeTransformOp_0 (Attribute%20Transform%20Op :InputSource1 $transform :InputSource2 $transform))
  (bind Select_0 (Select :bUseInputB-param UseParentAxis))
  (bind AttributeVectorOp_3 (Attribute%20Vector%20Op :InputSource1 $Position :Operation TransformLocation))
  (bind AttributeTransformOp_1 (Attribute%20Transform%20Op :InputSource1 $transform :InputSource2 $transform))
  (bind Select_9 (Select :bUseInputB-param UseParentAxis))
  ;[0.07297 0.5608 0.6146 0.5 30] SPLINE
  (bind AttributeTransformOp_10 (Attribute%20Transform%20Op :InputSource1 $Transform :Operation Invert))
  (bind ExtractAttribute_4 (Extract%20Attribute%20at%20Index :bForceOutputAttributeToBeInElementsDomain true :InputSource $Transform :OutputAttributeName trans))
  (bind AttributeTransformOp_9 (Attribute%20Transform%20Op :InputSource1 $transform :InputSource2 trans))
  (bind Select_1 (Select :bUseInputB-param UseParentAxis))
  (bind GetSplineControlPoints_0 (Get%20Spline%20Control%20Points))
  (bind AttributeTransformOp_2 (Attribute%20Transform%20Op :InputSource1 @Data.$SplineTransform))
  (bind ExtractAttribute_5 (Extract%20Attribute%20at%20Index :bForceOutputAttributeToBeInElementsDomain true :InputSource $Transform :OutputAttributeName trans))
  (bind AttributeTransformOp_4 (Attribute%20Transform%20Op :InputSource1 @Data.$SplineTransform))
  (bind Select_10 (Select :bUseInputB-param UseParentAxis))
  ;[0.15 0.15 0.15 0.5 70] PARENT_TRANSFORM
  (bind GetActorData_0 (Get%20Actor%20Data :Mode GetSinglePoint))
  (bind AttributeMathsOp_3 (Attribute%20Maths%20Op :InputSource1 $Scale :Operation Set))
  (reroute Center%20To%20Center AttributeMathsOp_18)
  (reroute Center%20to%20Max AttributeMathsOp_11)
  (reroute Center%20to%20Min AttributeMathsOp_10)
  (reroute Max%20to%20Center AttributeMathsOp_13)
  (reroute Max%20to%20Max AttributeMathsOp_19)
  (reroute Max%20to%20Min AttributeMathsOp_9)
  (reroute Min%20to%20Center AttributeMathsOp_12)
  (reroute Min%20to%20MAx AttributeMathsOp_7)
  (reroute Min%20to%20Min AttributeMathsOp_17)
  (reroute Parent%20Actor%20Transform AttributeMathsOp_3)
  (connect MergeAttributes_3 Out AttributeFilter_0 In)
  (connect MergeAttributes_4 Out AttributeFilter_1 In)
  (connect GetActorData_0 Out AttributeMathsOp_3 InA)
  (connect AttributeFilter_1 InsideFilter AttributeMathsOp_7 InA)
  (connect AttributeFilter_0 InsideFilter AttributeMathsOp_7 InB)
  (connect AttributeFilter_1 InsideFilter AttributeMathsOp_9 InA)
  (connect AttributeFilter_0 InsideFilter AttributeMathsOp_9 InB)
  (connect AttributeFilter_1 InsideFilter AttributeMathsOp_10 InA)
  (connect AttributeFilter_0 InsideFilter AttributeMathsOp_10 InB)
  (connect AttributeFilter_1 InsideFilter AttributeMathsOp_11 InA)
  (connect AttributeFilter_0 InsideFilter AttributeMathsOp_11 InB)
  (connect AttributeFilter_1 InsideFilter AttributeMathsOp_12 InA)
  (connect AttributeFilter_0 InsideFilter AttributeMathsOp_12 InB)
  (connect AttributeFilter_1 InsideFilter AttributeMathsOp_13 InA)
  (connect AttributeFilter_0 InsideFilter AttributeMathsOp_13 InB)
  (connect AttributeFilter_1 InsideFilter AttributeMathsOp_17 InA)
  (connect AttributeFilter_0 InsideFilter AttributeMathsOp_17 InB)
  (connect AttributeFilter_1 InsideFilter AttributeMathsOp_18 InA)
  (connect AttributeFilter_0 InsideFilter AttributeMathsOp_18 InB)
  (connect AttributeFilter_1 InsideFilter AttributeMathsOp_19 InA)
  (connect AttributeFilter_0 InsideFilter AttributeMathsOp_19 InB)
  (connect ExtractAttribute_11 Out AttributeMathsOp_21 InA)
  (connect AttributeRename_0 Out AttributeMathsOp_21 InB)
  (connect ExtractAttribute_10 Out AttributeMathsOp_22 InA)
  (connect AttributeRename_0 Out AttributeMathsOp_22 InB)
  (connect ExtractAttribute_8 Out AttributeMathsOp_23 InA)
  (connect AttributeRename_1 Out AttributeMathsOp_23 InB)
  (connect ExtractAttribute_7 Out AttributeMathsOp_24 InA)
  (connect AttributeRename_1 Out AttributeMathsOp_24 InB)
  (connect ExtractAttribute_9 Out AttributeRename_0 In)
  (connect ExtractAttribute_3 Out AttributeRename_1 In)
  (connect FilterDataByType_1 InsideFilter AttributeTransformOp_0 InA)
  (connect AttributeTransformOp_3 Out AttributeTransformOp_0 InB)
  (connect AttributeVectorOp_3 Out AttributeTransformOp_1 InA)
  (connect Parent%20Actor%20Transform Out AttributeTransformOp_1 InB)
  (connect Select_1 Out AttributeTransformOp_2 InA)
  (connect MakeTransformAttribute_1 Out AttributeTransformOp_2 InB)
  (connect Parent%20Actor%20Transform Out AttributeTransformOp_3 In)
  (connect AttributeTransformOp_2 Out AttributeTransformOp_4 InA)
  (connect ExtractAttribute_5 Out AttributeTransformOp_4 InB)
  (connect AttributeVectorOp_2 Out AttributeTransformOp_8 InA)
  (connect Parent%20Actor%20Transform Out AttributeTransformOp_8 InB)
  (connect FilterDataByType_2 InsideFilter AttributeTransformOp_9 InA)
  (connect ExtractAttribute_4 Out AttributeTransformOp_9 InB)
  (connect Parent%20Actor%20Transform Out AttributeTransformOp_10 In)
  (connect ToPoint_1 Out AttributeTransformOp_13 InA)
  (connect AttributeTransformOp_14 Out AttributeTransformOp_13 InB)
  (connect Parent%20Actor%20Transform Out AttributeTransformOp_14 In)
  (connect FilterDataByType_5 InsideFilter AttributeTransformOp_15 InA)
  (connect GetSplineControlPoints_4 Out AttributeTransformOp_15 InA)
  (connect ToPoint_0 Out AttributeTransformOp_15 InA)
  (connect AttributeTransformOp_16 Out AttributeTransformOp_15 InB)
  (connect Parent%20Actor%20Transform Out AttributeTransformOp_16 In)
  (connect Select_3 Out AttributeVectorOp_2 In)
  (connect MakeTransformAttribute_1 Out AttributeVectorOp_2 Transform)
  (connect Select_0 Out AttributeVectorOp_3 In)
  (connect MakeTransformAttribute_1 Out AttributeVectorOp_3 Transform)
  (connect MergePoints_2 Out CombinePoints_0 In)
  (connect MergePoints_1 Out CombinePoints_1 In)
  (connect Select_12 Out CreatePolygon2D_1 In)
  (connect Subgraph_3 Out ExtractAttribute_0 In)
  (connect Subgraph_4 Out ExtractAttribute_1 In)
  (connect Subgraph_5 Out ExtractAttribute_2 In)
  (connect CombinePoints_0 Out ExtractAttribute_3 In)
  (connect AttributeTransformOp_10 Out ExtractAttribute_4 In)
  (connect Parent%20Actor%20Transform Out ExtractAttribute_5 In)
  (connect CombinePoints_0 Out ExtractAttribute_7 In)
  (connect CombinePoints_0 Out ExtractAttribute_8 In)
  (connect CombinePoints_1 Out ExtractAttribute_9 In)
  (connect CombinePoints_1 Out ExtractAttribute_10 In)
  (connect CombinePoints_1 Out ExtractAttribute_11 In)
  (connect DefaultInputNode In FilterDataByTag_1 In)
  (connect FilterDataByTag_1 InsideFilter FilterDataByType_1 In)
  (connect FilterDataByTag_1 InsideFilter FilterDataByType_2 In)
  (connect FilterDataByTag_1 InsideFilter FilterDataByType_3 In)
  (connect FilterDataByTag_1 OutsideFilter FilterDataByType_5 In)
  (connect FilterDataByTag_1 OutsideFilter FilterDataByType_6 In)
  (connect FilterDataByTag_1 OutsideFilter FilterDataByType_7 In)
  (connect Select_1 Out GetSplineControlPoints_0 In)
  (connect FilterDataByType_6 InsideFilter GetSplineControlPoints_4 In)
  (connect CreateConstant_4 Out MakeTransformAttribute_1 Scale)
  (connect MakeVectorAttribute_2 Out MakeTransformAttribute_1 Translation)
  (connect Select_4 Out MakeVectorAttribute_2 X)
  (connect Select_5 Out MakeVectorAttribute_2 Y)
  (connect Select_6 Out MakeVectorAttribute_2 Z)
  (connect AttributeMathsOp_23 Out MergeAttributes_3 In)
  (connect AttributeMathsOp_24 Out MergeAttributes_3 In)
  (connect AttributeRename_1 Out MergeAttributes_3 In)
  (connect AttributeMathsOp_21 Out MergeAttributes_4 In)
  (connect AttributeMathsOp_22 Out MergeAttributes_4 In)
  (connect AttributeRename_0 Out MergeAttributes_4 In)
  (connect Select_8 Out MergePoints_1 In)
  (connect GetSplineControlPoints_0 Out MergePoints_2 In)
  (connect Select_0 Out MergePoints_2 In)
  (connect Select_3 Out MergePoints_2 In)
  (connect FilterDataByType_1 InsideFilter Select_0 Input%20A)
  (connect AttributeTransformOp_0 Out Select_0 Input%20B)
  (connect FilterDataByType_2 InsideFilter Select_1 Input%20A)
  (connect AttributeTransformOp_9 Out Select_1 Input%20B)
  (connect ToPoint_1 Out Select_3 Input%20A)
  (connect AttributeTransformOp_13 Out Select_3 Input%20B)
  (connect AttributeMathsOp_8 Out Select_4 bUseInputB)
  (connect CreateConstant_5 Out Select_4 Input%20A)
  (connect ExtractAttribute_0 Out Select_4 Input%20B)
  (connect AttributeMathsOp_14 Out Select_5 bUseInputB)
  (connect CreateConstant_6 Out Select_5 Input%20A)
  (connect ExtractAttribute_1 Out Select_5 Input%20B)
  (connect AttributeMathsOp_15 Out Select_6 bUseInputB)
  (connect CreateConstant_7 Out Select_6 Input%20A)
  (connect ExtractAttribute_2 Out Select_6 Input%20B)
  (connect FilterDataByType_5 InsideFilter Select_8 Input%20A)
  (connect GetSplineControlPoints_4 Out Select_8 Input%20A)
  (connect ToPoint_0 Out Select_8 Input%20A)
  (connect AttributeTransformOp_15 Out Select_8 Input%20B)
  (connect AttributeVectorOp_3 Out Select_9 Input%20A)
  (connect AttributeTransformOp_1 Out Select_9 Input%20B)
  (connect AttributeTransformOp_2 Out Select_10 Input%20A)
  (connect AttributeTransformOp_4 Out Select_10 Input%20B)
  (connect AttributeVectorOp_2 Out Select_12 Input%20A)
  (connect AttributeTransformOp_8 Out Select_12 Input%20B)
  (connect Center%20To%20Center Out Subgraph_3 Center%20to%20Center%20Out)
  (connect Center%20to%20Max Out Subgraph_3 Center%20To%20Max%20Out)
  (connect Center%20to%20Min Out Subgraph_3 Center%20To%20Min%20Out)
  (connect Max%20to%20Center Out Subgraph_3 Max%20To%20Center%20Out)
  (connect Max%20to%20Max Out Subgraph_3 Max%20To%20Max%20Out)
  (connect Max%20to%20Min Out Subgraph_3 Max%20To%20Min%20Out)
  (connect Min%20to%20Center Out Subgraph_3 Min%20To%20Center%20Out)
  (connect Min%20to%20MAx Out Subgraph_3 Min%20To%20MAx%20Out)
  (connect Min%20to%20Min Out Subgraph_3 Out)
  (connect Center%20To%20Center Out Subgraph_4 Center%20to%20Center%20Out)
  (connect Center%20to%20Max Out Subgraph_4 Center%20To%20Max%20Out)
  (connect Center%20to%20Min Out Subgraph_4 Center%20To%20Min%20Out)
  (connect Max%20to%20Center Out Subgraph_4 Max%20To%20Center%20Out)
  (connect Max%20to%20Max Out Subgraph_4 Max%20To%20Max%20Out)
  (connect Max%20to%20Min Out Subgraph_4 Max%20To%20Min%20Out)
  (connect Min%20to%20Center Out Subgraph_4 Min%20To%20Center%20Out)
  (connect Min%20to%20MAx Out Subgraph_4 Min%20To%20MAx%20Out)
  (connect Min%20to%20Min Out Subgraph_4 Out)
  (connect Center%20To%20Center Out Subgraph_5 Center%20to%20Center%20Out)
  (connect Center%20to%20Max Out Subgraph_5 Center%20To%20Max%20Out)
  (connect Center%20to%20Min Out Subgraph_5 Center%20To%20Min%20Out)
  (connect Max%20to%20Center Out Subgraph_5 Max%20To%20Center%20Out)
  (connect Max%20to%20Max Out Subgraph_5 Max%20To%20Max%20Out)
  (connect Max%20to%20Min Out Subgraph_5 Max%20To%20Min%20Out)
  (connect Min%20to%20Center Out Subgraph_5 Min%20To%20Center%20Out)
  (connect Min%20to%20MAx Out Subgraph_5 Min%20To%20MAx%20Out)
  (connect Min%20to%20Min Out Subgraph_5 Out)
  (connect FilterDataByType_7 InsideFilter ToPoint_0 In)
  (connect FilterDataByType_3 InsideFilter ToPoint_1 In)
  (output Out CreatePolygon2D_1)
  (output Out Select_9)
  (output Out Select_10))
