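; PCG subgraph that serves as the core process for contracting 2D polygon areas by a given distance.
; It keeps only Polygon2D data from the input and offsets it by the negated Contract value, so a positive value shrinks each area.
; When Union is true (the default), a Branch sends the offset polygons through a union operation; otherwise they go straight to Out.
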(graph Transform_Areas_Contract_CoreProcess
  (params (Contract :type Double :default 1e+02) (Union :type Bool :default true))
  ;[0 0 0 0.5] Core Process
  (bind AttributeMathsOp_0 (Attribute%20Maths%20Op :Operation Negate :In Contract))
  (bind FilterDataByType_0 (Filter%20Data%20By%20Type :TargetType (struct :Ids [(struct :Struct /Script/PCG.PCGDataTypeInfoPolygon2D)])))
  (bind OffsetPolygon2D_0 (Offset%20Polygon))
  (bind Branch_0 (Branch :bOutputToB-param Union))
  (bind Polygon2DOperation_1 (Polygon%20Operation :Operation Union))
  (connect OffsetPolygon2D_0 Out Branch_0 In)
  (connect DefaultInputNode In FilterDataByType_0 In)
  (connect FilterDataByType_0 InsideFilter OffsetPolygon2D_0 In)
  (connect AttributeMathsOp_0 Out OffsetPolygon2D_0 Offset)
  (connect Branch_0 Output%20B Polygon2DOperation_1 In)
  (output Out Branch_0.Output%20A)
  (output Out Polygon2DOperation_1))
