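; Transform_Areas_Expand_CoreProcess is a PCG subgraph that expands 2D polygon areas by a configurable distance, optionally merging the results.
; It keeps only Polygon2D data from its input and feeds it to an Offset Polygon node driven by the Expand parameter.
; A Branch on the Union flag either outputs the offset polygons directly or passes them through a Union polygon operation first.
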(graph Transform_Areas_Expand_CoreProcess
  (params (Expand :type Double :default 1e+02) (Union :type Bool))
  ;[0 0 0 0.5] Core Process
  (bind FilterDataByType_0 (Filter%20Data%20By%20Type :TargetType (struct :Ids [(struct :Struct /Script/PCG.PCGDataTypeInfoPolygon2D)])))
  (bind OffsetPolygon2D_0 (Offset%20Polygon :Offset-param Expand))
  (bind Branch_1 (Branch :bOutputToB-param Union))
  (bind Polygon2DOperation_0 (Polygon%20Operation :Operation Union))
  (connect OffsetPolygon2D_0 Out Branch_1 In)
  (connect DefaultInputNode In FilterDataByType_0 In)
  (connect FilterDataByType_0 InsideFilter OffsetPolygon2D_0 In)
  (connect Branch_1 Output%20B Polygon2DOperation_0 In)
  (output Out Branch_1.Output%20A)
  (output Out Polygon2DOperation_0))
