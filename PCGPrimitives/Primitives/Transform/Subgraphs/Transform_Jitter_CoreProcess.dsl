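; PCG subgraph that applies seeded positional jitter to point, spline and Polygon2D data, as the core of a Transform Jitter primitive.
; Input is split by type, then Transform Points offsets points, spline control points and polygon points using bounds derived from the jitter vector.
; Jittered positions are copied back onto source splines, which are rebuilt as open or closed with custom tangents and interp types.
; Polygons are recreated from their jittered points.
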(graph Transform_Jitter_CoreProcess
  (params (jitter :type Vector) (seed :type Int))
  (bind FilterDataByType_0 (Filter%20Data%20By%20Type :TargetType (struct :Ids [(struct :Struct /Script/PCG.PCGDataTypeInfoPoint)])))
  (bind FilterDataByType_1 (Filter%20Data%20By%20Type :TargetType (struct :Ids [(struct :Struct /Script/PCG.PCGDataTypeInfoSpline)])))
  (bind FilterDataByType_2 (Filter%20Data%20By%20Type :TargetType (struct :Ids [(struct :Struct /Script/PCG.PCGDataTypeInfoPolygon2D)])))
  (bind GetSplineControlPoints_0 (Get%20Spline%20Control%20Points))
  (bind AttributeMathsOp_0 (Attribute%20Maths%20Op :Operation Multiply :InA jitter))
  (bind CopyAttributes_0 (Copy%20Attributes :InputSource $Position :OutputTarget $Position))
  (bind ToPoint_0 (To%20Point))
  (bind FilterDataByAttribute_0 (Filter%20Data%20By%20Attribute :FilterMode FilterByValue :FilterOperator Equal :TargetAttribute @Data.$isClosed :Threshold (struct :AttributeTypes (struct :Type Boolean :BoolValue true))))
  ;[0 0 0 0.5] Jitter
  (bind TransformPoints_0 (Transform%20Points :OffsetMin-param jitter :Seed-param seed))
  (bind TransformPoints_1 (Transform%20Points :OffsetMin-param jitter :Seed-param seed))
  (bind TransformPoints_2 (Transform%20Points :OffsetMin-param jitter :Seed-param seed))
  ;[0 0 0 0.5] Reconstruct Poly
  (bind CreatePolygon2D_0 (Create%20Polygon%202D))
  ;[0 0 0 0.5] Recompute splines
  (bind GetSplineControlPoints_1 (Get%20Spline%20Control%20Points))
  (bind CreateSpline_1 (Create%20Spline :bApplyCustomTangents true :bUseInterpTypeAttribute true))
  (bind GetSplineControlPoints_2 (Get%20Spline%20Control%20Points))
  (bind CreateSpline_2 (Create%20Spline :bApplyCustomTangents true :bClosedLoop true :bUseInterpTypeAttribute true))
  (connect TransformPoints_0 Out CopyAttributes_0 Source)
  (connect FilterDataByType_1 InsideFilter CopyAttributes_0 Target)
  (connect TransformPoints_2 Out CreatePolygon2D_0 In)
  (connect GetSplineControlPoints_1 Out CreateSpline_1 In)
  (connect GetSplineControlPoints_2 Out CreateSpline_2 In)
  (connect CopyAttributes_0 Out FilterDataByAttribute_0 In)
  (connect DefaultInputNode In FilterDataByType_0 In)
  (connect DefaultInputNode In FilterDataByType_1 In)
  (connect DefaultInputNode In FilterDataByType_2 In)
  (connect FilterDataByType_1 InsideFilter GetSplineControlPoints_0 In)
  (connect FilterDataByAttribute_0 OutsideFilter GetSplineControlPoints_1 In)
  (connect FilterDataByAttribute_0 InsideFilter GetSplineControlPoints_2 In)
  (connect FilterDataByType_2 InsideFilter ToPoint_0 In)
  (connect GetSplineControlPoints_0 Out TransformPoints_0 In)
  (connect AttributeMathsOp_0 Out TransformPoints_0 OffsetMax)
  (connect FilterDataByType_0 InsideFilter TransformPoints_1 In)
  (connect AttributeMathsOp_0 Out TransformPoints_1 OffsetMax)
  (connect ToPoint_0 Out TransformPoints_2 In)
  (connect AttributeMathsOp_0 Out TransformPoints_2 OffsetMax)
  (output Out CreatePolygon2D_0)
  (output Out CreateSpline_1)
  (output Out CreateSpline_2)
  (output Out TransformPoints_1))
