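; PCG subgraph for the Level transform primitive that flattens each input point data to a single height.
; It keeps only data tagged Input:0, the primary input, through a tag filter.
; Attribute Reduce computes one Position.z per data using the LevelOperation enum parameter, such as average, min or max.
; A Set maths op writes that value back to every point's Z and outputs the result.
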(graph Transform_Level_CoreProcess
  (params (LevelOperation :type (Enum /PCGPrimitives/Primitives/Transform/Resources/EPCG_LevelEnum.EPCG_LevelEnum)))
  ;[0 0 0 0.5 12] Primary/Secondary Inputs
  (bind FilterDataByTag_0 (Filter%20Data%20By%20Tag :SelectedTags "Input:0"))
  ;[0 0 0 0.5] Core Process
  (bind AttributeReduce_0 (Attribute%20Reduce :InputSource $Position.z :Operation-param LevelOperation)) ; Find, average, max, min position from all elements\nfor each input data
  (bind AttributeMathsOp_0 (Attribute%20Maths%20Op :InputSource1 $Position.z :Operation Set)) ; Re-apply the value on all elements
  (connect FilterDataByTag_0 InsideFilter AttributeMathsOp_0 InA)
  (connect AttributeReduce_0 Out AttributeMathsOp_0 InB)
  (connect FilterDataByTag_0 InsideFilter AttributeReduce_0 In)
  (connect DefaultInputNode In FilterDataByTag_0 In)
  (output Out AttributeMathsOp_0))
