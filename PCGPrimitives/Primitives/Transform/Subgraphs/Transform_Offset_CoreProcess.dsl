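; PCG subgraph that applies offset, rotation and scale to points, splines and Polygon2Ds, split by data type.
; Per-point mode transforms points and polygon control points, rebuilding polygons; splines receive only translation, via Offset Spline.
; An optional attribute mask lerps points and polygons toward their originals and scales spline offset magnitude.
; Global mode applies one transform in parent-actor space by inverting the actor transform, applying the offset, then restoring it.
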(graph Transform_Offset_CoreProcess
  (params (offset :type Vector) (rotation_offset :type Rotator) (scale :type Vector :default (1.0 1.0 1.0)) (absolute_offset :type Bool) (Per_Points :type Bool :default true) (UseAttributeMask :type Bool) (AttributeMask :type String :default "@Last"))
  (bind Branch_1 (Branch :bOutputToB-param Per_Points))
  ;[0 0 0 0.5] Per Point Transform
  (bind Branch_7 (Branch :bOutputToB-param UseAttributeMask))
  ;;[0.15 0.15 0.15 0.5] Points
  (bind FilterDataByType_0 (Filter%20Data%20By%20Type :TargetType (struct :Ids [(struct :Struct /Script/PCG.PCGDataTypeInfoPoint)])))
  (bind TransformPoints_0 (Transform%20Points :bAbsoluteOffset-param absolute_offset :bAbsoluteRotation-param absolute_offset :bAbsoluteScale-param absolute_offset :OffsetMax-param offset :OffsetMin-param offset :RotationMax-param rotation_offset :RotationMin-param rotation_offset :ScaleMax-param scale :ScaleMin-param scale))
  (bind Branch_3 (Branch :bOutputToB-param UseAttributeMask))
  (bind AttributeMathsOp_4 (Attribute%20Maths%20Op :Operation OneMinus :InputSource1-param AttributeMask))
  (bind Branch_0 (Branch :bOutputToB-param UseAttributeMask))
  (bind AttributeMathsOp_1 (Attribute%20Maths%20Op :InputSource1 $Position :InputSource2 $Position :InputSource3 Dirt :Operation Lerp :InputSource3-param AttributeMask))
  (bind AttributeMathsOp_2 (Attribute%20Maths%20Op :InputSource1 $Rotation.EulerAngles :InputSource2 $Rotation.EulerAngles :InputSource3 Dirt :Operation Lerp :InputSource3-param AttributeMask))
  (bind AttributeMathsOp_3 (Attribute%20Maths%20Op :InputSource1 $Scale :InputSource2 $Scale :InputSource3 Dirt :Operation Lerp :InputSource3-param AttributeMask))
  ;;[0.15 0.15 0.15 0.5] Spline
  (bind FilterDataByType_1 (Filter%20Data%20By%20Type :TargetType (struct :Ids [(struct :Struct /Script/PCG.PCGDataTypeInfoSpline)])))
  (bind GetSplineControlPoints_1 (Get%20Spline%20Control%20Points))
  (bind TransformPoints_1 (Transform%20Points :OffsetMax-param offset :OffsetMin-param offset))
  (bind AttributeMathsOp_13 (Attribute%20Maths%20Op :InputSource1 $Position :InputSource2 $Position :Operation Subtract :OutputTarget Direction))
  (bind CopyAttributes_0 (Copy%20Attributes :InputSource Direction))
  (bind CreateConstant_0 (Create%20Constant :AttributeTypes (struct :Type Float :FloatValue 1.0)))
  (bind AddAttribute_1 (Add%20Attribute :OutputTarget Magnitude))
  (bind Branch_2 (Branch :bOutputToB-param UseAttributeMask))
  (bind AttributeMathsOp_12 (Attribute%20Maths%20Op :InputSource1 Magnitude :Operation Multiply :InputSource2-param AttributeMask))
  (bind OffsetSpline_0 (Offset%20Spline :DirectionAttribute Direction :MagnitudeAttribute MAgnitude))
  ;;[0.15 0.15 0.15 0.5] Polygon2D - individual polygon2D will remain flat however disregarding fluctuating attribute mask
  (bind FilterDataByType_2 (Filter%20Data%20By%20Type :TargetType (struct :Ids [(struct :Struct /Script/PCG.PCGDataTypeInfoPolygon2D)])))
  (bind GetSplineControlPoints_5 (Get%20Spline%20Control%20Points))
  (bind AttributeMathsOp_11 (Attribute%20Maths%20Op :Operation OneMinus :InputSource1-param AttributeMask))
  (bind TransformPoints_2 (Transform%20Points :bAbsoluteOffset-param absolute_offset :bAbsoluteRotation-param absolute_offset :bAbsoluteScale-param absolute_offset :OffsetMax-param offset :OffsetMin-param offset :RotationMax-param rotation_offset :RotationMin-param rotation_offset :ScaleMax-param scale :ScaleMin-param scale))
  (bind Branch_6 (Branch :bOutputToB-param UseAttributeMask))
  (bind AttributeMathsOp_0 (Attribute%20Maths%20Op :InputSource1 $Position :InputSource2 $Position :InputSource3 Dirt :Operation Lerp :InputSource3-param AttributeMask))
  (bind AttributeMathsOp_9 (Attribute%20Maths%20Op :InputSource1 $Rotation.EulerAngles :InputSource2 $Rotation.EulerAngles :InputSource3 Dirt :Operation Lerp :InputSource3-param AttributeMask))
  (bind AttributeMathsOp_10 (Attribute%20Maths%20Op :InputSource1 $Scale :InputSource2 $Scale :InputSource3 Dirt :Operation Lerp :InputSource3-param AttributeMask))
  (bind CreatePolygon2D_0 (Create%20Polygon%202D))
  ;[0 0 0 0.5] Global Transform
  (bind FilterDataByType_3 (Filter%20Data%20By%20Type :TargetType (struct :Ids [(struct :Struct /Script/PCG.PCGDataTypeInfoPoint)])))
  (bind FilterDataByType_4 (Filter%20Data%20By%20Type :TargetType (struct :Ids [(struct :Struct /Script/PCG.PCGDataTypeInfoSpline)])))
  (bind FilterDataByType_5 (Filter%20Data%20By%20Type :TargetType (struct :Ids [(struct :Struct /Script/PCG.PCGDataTypeInfoPolygon2D)])))
  (bind MakeTransformAttribute_0 (Make%20Transform%20Attribute :Rotation rotation_offset :Scale scale :Translation offset))
  (bind AttributeTransformOp_1 (Attribute%20Transform%20Op :InputSource1 $transform :InputSource2 trans))
  (bind AttributeTransformOp_0 (Attribute%20Transform%20Op :InputSource1 $Transform))
  (bind AttributeTransformOp_3 (Attribute%20Transform%20Op :InputSource1 @data.$SplineTransform :InputSource2 trans))
  (bind AttributeTransformOp_4 (Attribute%20Transform%20Op :InputSource1 $Transform))
  (bind AttributeTransformOp_11 (Attribute%20Transform%20Op :InputSource1 @Data.$SplineTransform :InputSource2 trans))
  (bind AttributeTransformOp_12 (Attribute%20Transform%20Op :InputSource1 $transform :InputSource2 trans))
  (bind AttributeTransformOp_5 (Attribute%20Transform%20Op :InputSource1 @data.$Transform :InputSource2 trans))
  (bind ToPoint_0 (To%20Point))
  (bind AttributeTransformOp_6 (Attribute%20Transform%20Op :InputSource1 $Transform))
  (bind CreatePolygon2D_1 (Create%20Polygon%202D))
  (bind AttributeTransformOp_10 (Attribute%20Transform%20Op :InputSource1 @Data.$Transform :InputSource2 trans))
  (bind AttributeTransformOp_8 (Attribute%20Transform%20Op))
  ;[0.15 0.15 0.15 0.5] PARENT transform
  (bind GetActorData_1 (Get%20Actor%20Data :Mode GetSinglePoint))
  (bind AttributeMathsOp_5 (Attribute%20Maths%20Op :InputSource1 $Scale :Operation Set))
  (bind AttributeTransformOp_14 (Attribute%20Transform%20Op :InputSource1 $Transform :Operation Invert))
  (bind ExtractAttribute_0 (Extract%20Attribute%20at%20Index :bForceOutputAttributeToBeInElementsDomain true :InputSource $Transform :OutputAttributeName trans))
  (bind ExtractAttribute_1 (Extract%20Attribute%20at%20Index :bForceOutputAttributeToBeInElementsDomain true :InputSource $Transform :OutputAttributeName trans))
  (reroute invert%20parent%20transform ExtractAttribute_0)
  (reroute Parent%20Actor%20Transform_1 AttributeMathsOp_5)
  (reroute parent%20transform ExtractAttribute_1)
  (connect CreateConstant_0 Out AddAttribute_1 Attributes)
  (connect CopyAttributes_0 Out AddAttribute_1 In)
  (connect Branch_6 Output%20B AttributeMathsOp_0 InA)
  (connect AttributeMathsOp_11 Out AttributeMathsOp_0 InB)
  (connect AttributeMathsOp_11 Out AttributeMathsOp_0 Ratio)
  (connect Branch_0 Output%20B AttributeMathsOp_1 InA)
  (connect AttributeMathsOp_4 Out AttributeMathsOp_1 InB)
  (connect AttributeMathsOp_4 Out AttributeMathsOp_1 Ratio)
  (connect AttributeMathsOp_1 Out AttributeMathsOp_2 InA)
  (connect AttributeMathsOp_4 Out AttributeMathsOp_2 InB)
  (connect AttributeMathsOp_4 Out AttributeMathsOp_2 Ratio)
  (connect AttributeMathsOp_2 Out AttributeMathsOp_3 InA)
  (connect AttributeMathsOp_4 Out AttributeMathsOp_3 InB)
  (connect AttributeMathsOp_4 Out AttributeMathsOp_3 Ratio)
  (connect Branch_3 Output%20B AttributeMathsOp_4 In)
  (connect GetActorData_1 Out AttributeMathsOp_5 InA)
  (connect AttributeMathsOp_0 Out AttributeMathsOp_9 InA)
  (connect AttributeMathsOp_11 Out AttributeMathsOp_9 InB)
  (connect AttributeMathsOp_11 Out AttributeMathsOp_9 Ratio)
  (connect AttributeMathsOp_9 Out AttributeMathsOp_10 InA)
  (connect AttributeMathsOp_11 Out AttributeMathsOp_10 InB)
  (connect AttributeMathsOp_11 Out AttributeMathsOp_10 Ratio)
  (connect Branch_7 Output%20B AttributeMathsOp_11 In)
  (connect Branch_2 Output%20B AttributeMathsOp_12 InA)
  (connect Branch_2 Output%20B AttributeMathsOp_12 InB)
  (connect TransformPoints_1 Out AttributeMathsOp_13 InA)
  (connect GetSplineControlPoints_1 Out AttributeMathsOp_13 InB)
  (connect AttributeTransformOp_1 Out AttributeTransformOp_0 InA)
  (connect MakeTransformAttribute_0 Out AttributeTransformOp_0 InB)
  (connect FilterDataByType_3 InsideFilter AttributeTransformOp_1 InA)
  (connect invert%20parent%20transform Out AttributeTransformOp_1 InB)
  (connect FilterDataByType_4 InsideFilter AttributeTransformOp_3 InA)
  (connect invert%20parent%20transform Out AttributeTransformOp_3 InB)
  (connect AttributeTransformOp_3 Out AttributeTransformOp_4 InA)
  (connect MakeTransformAttribute_0 Out AttributeTransformOp_4 InB)
  (connect FilterDataByType_5 InsideFilter AttributeTransformOp_5 InA)
  (connect invert%20parent%20transform Out AttributeTransformOp_5 InB)
  (connect ToPoint_0 Out AttributeTransformOp_6 InA)
  (connect MakeTransformAttribute_0 Out AttributeTransformOp_6 InB)
  (connect MakeTransformAttribute_0 Out AttributeTransformOp_8 InA)
  (connect parent%20transform Out AttributeTransformOp_8 InB)
  (connect CreatePolygon2D_1 Out AttributeTransformOp_10 InA)
  (connect parent%20transform Out AttributeTransformOp_10 InB)
  (connect AttributeTransformOp_4 Out AttributeTransformOp_11 InA)
  (connect parent%20transform Out AttributeTransformOp_11 InB)
  (connect AttributeTransformOp_0 Out AttributeTransformOp_12 InA)
  (connect parent%20transform Out AttributeTransformOp_12 InB)
  (connect Parent%20Actor%20Transform_1 Out AttributeTransformOp_14 In)
  (connect TransformPoints_0 Out Branch_0 In)
  (connect DefaultInputNode In Branch_1 In)
  (connect AddAttribute_1 Out Branch_2 In)
  (connect FilterDataByType_0 InsideFilter Branch_3 In)
  (connect TransformPoints_2 Out Branch_6 In)
  (connect GetSplineControlPoints_5 Out Branch_7 In)
  (connect AttributeMathsOp_13 Out CopyAttributes_0 Source)
  (connect FilterDataByType_1 InsideFilter CopyAttributes_0 Target)
  (connect AttributeMathsOp_10 Out CreatePolygon2D_0 In)
  (connect Branch_6 Output%20A CreatePolygon2D_0 In)
  (connect AttributeTransformOp_6 Out CreatePolygon2D_1 In)
  (connect AttributeTransformOp_14 Out ExtractAttribute_0 In)
  (connect Parent%20Actor%20Transform_1 Out ExtractAttribute_1 In)
  (connect Branch_1 Output%20B FilterDataByType_0 In)
  (connect Branch_1 Output%20B FilterDataByType_1 In)
  (connect Branch_1 Output%20B FilterDataByType_2 In)
  (connect Branch_1 Output%20A FilterDataByType_3 In)
  (connect Branch_1 Output%20A FilterDataByType_4 In)
  (connect Branch_1 Output%20A FilterDataByType_5 In)
  (connect FilterDataByType_1 InsideFilter GetSplineControlPoints_1 In)
  (connect FilterDataByType_2 InsideFilter GetSplineControlPoints_5 In)
  (connect CopyAttributes_0 Out OffsetSpline_0 In)
  (connect AttributeMathsOp_12 Out OffsetSpline_0 Magnitude)
  (connect Branch_2 Output%20A OffsetSpline_0 Magnitude)
  (connect AttributeTransformOp_5 Out ToPoint_0 In)
  (connect FilterDataByType_0 InsideFilter TransformPoints_0 In)
  (connect GetSplineControlPoints_1 Out TransformPoints_1 In)
  (connect GetSplineControlPoints_5 Out TransformPoints_2 In)
  (output Out AttributeMathsOp_3)
  (output Out AttributeTransformOp_10)
  (output Out AttributeTransformOp_11)
  (output Out AttributeTransformOp_12)
  (output Out Branch_0.Output%20A)
  (output Out CreatePolygon2D_0)
  (output Out OffsetSpline_0))
