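; PCG subgraph that turns input splines into outline polygons of a given width.
; It filters to splines and splits them by $IsClosed: open ones become width-thick strokes directly through Create Polygon 2D.
; Closed ones become closed polygons, offset outward and inward by half the width, with the inner subtracted from the outer to form a ring.
; Both branches feed the same output.
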(graph Transform_Outline_CoreProcess
  (params (width :type Double :default 1e+02))
  (bind FilterDataByType_0 (Filter%20Data%20By%20Type :TargetType (struct :Ids [(struct :Struct /Script/PCG.PCGDataTypeInfoSpline)])))
  (bind FilterDataByAttribute_0 (Filter%20Data%20By%20Attribute :FilterMode FilterByValue :FilterOperator Equal :TargetAttribute @data.$IsClosed :Threshold (struct :AttributeTypes (struct :Type Boolean)))) ; Splits splines into open and closed using the $IsClosed attribute.\nOpen splines go to the top path, closed splines to the bottom path.
  (bind CreatePolygon2D_0 (Create%20Polygon%202D :InputType ForceOpen :OpenPolygonWidth-param width)) ; Open splines are handled natively by Create Polygon 2D.\nThe width parameter drives the stroke thickness directly.
  (bind AttributeMathsOp_0 (Attribute%20Maths%20Op :Operation Multiply :InA width)) ; Multiplies the width by half to get the offset amount.\nOne instance pushes outward, the other pulls inward,\ngiving the outline its thickness.
  (bind CreatePolygon2D_2 (Create%20Polygon%202D :InputType ForceClosed)) ; Closed splines cannot use the open polygon stroke trick.\nThey are forced into a closed polygon so we can\napply the half-width push and pull offset technique below.
  (bind OffsetPolygon2D_0 (Offset%20Polygon))
  (bind AttributeMathsOp_1 (Attribute%20Maths%20Op :Operation Multiply :InA width)) ; Multiplies the width by half to get the offset amount.\nOne instance pushes outward, the other pulls inward,\ngiving the outline its thickness.
  (bind OffsetPolygon2D_1 (Offset%20Polygon))
  (bind Polygon2DOperation_0 (Polygon%20Operation :Operation Difference)) ; Subtracts the inner offset polygon from the outer one.\nThis produces the final outline ring with the correct thickness.
  (connect FilterDataByAttribute_0 InsideFilter CreatePolygon2D_0 In)
  (connect FilterDataByAttribute_0 OutsideFilter CreatePolygon2D_2 In)
  (connect FilterDataByType_0 InsideFilter FilterDataByAttribute_0 In)
  (connect DefaultInputNode In FilterDataByType_0 In)
  (connect CreatePolygon2D_2 Out OffsetPolygon2D_0 In)
  (connect AttributeMathsOp_0 Out OffsetPolygon2D_0 Offset)
  (connect CreatePolygon2D_2 Out OffsetPolygon2D_1 In)
  (connect AttributeMathsOp_1 Out OffsetPolygon2D_1 Offset)
  (connect OffsetPolygon2D_1 Out Polygon2DOperation_0 ClipPolygons)
  (connect OffsetPolygon2D_0 Out Polygon2DOperation_0 In)
  (output Out CreatePolygon2D_0)
  (output Out Polygon2DOperation_0))
