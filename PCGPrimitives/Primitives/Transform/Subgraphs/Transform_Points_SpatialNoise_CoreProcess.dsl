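; PCG subgraph that displaces input points vertically with tiling Caustic2D spatial noise, scaled by an amplitude and added to Position.Z.
; Optional domain warping first shifts the sample positions in X and Y with two identically configured noise passes, while output points keep their original XY.
; The height noise is optionally blurred before being applied.
; Parameters expose amplitude, iterations, offset and scale for both noises, plus blur settings.
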(graph Transform_Points_SpatialNoise_CoreProcess
  (params (NoiseAmplitude :type Double :default 4e+03) (NoiseIterations :type Int :default 3) (NoiseOffset :type Vector) (NoiseScale :type Vector :default (1.0 1.0 1.0)) (NoiseBlur :type Bool :default true) (NoiseBlurIterations :type Int :default 2) (NoiseBlurDistance :type Double :default 1e+03) (Distort :type Bool :default true) (DistortAmplitude :type Double :default 5e+03) (DistortIterations :type Int :default 3) (DistortOffset :type Vector) (DistortScale :type Vector :default (2.0 2.0 2.0)))
  (bind FilterDataByType_1 (Filter%20Data%20By%20Type :TargetType (struct :Ids [(struct :Struct /Script/PCG.PCGDataTypeInfoPoint)])))
  (bind AttributeMathsOp_1 (Attribute%20Maths%20Op :InputSource1 $Position.Z :InputSource2 Noise :OutputTarget $Position.Z))
  ;[0 0 0 0.5] Noise
  (bind MakeTransformAttribute_1 (Make%20Transform%20Attribute :Scale NoiseScale))
  (bind TransformPoints_0 (Transform%20Points :bUniformScale false :OffsetMax-param NoiseOffset :OffsetMin-param NoiseOffset))
  (bind Spatial%20Noise_1 (Spatial%20Noise :bTiling true :Contrast 2.0 :Iterations 3 :Mode Caustic2D :RandomOffset (0.0 0.0 0.0) :Transform (struct :Scale3D (0.5 0.5 0.5)) :ValueTarget Noise :VoronoiCellRandomness 0.3608 :Iterations-param NoiseIterations))
  (bind AttributeMathsOp_0 (Attribute%20Maths%20Op :InputSource1 Noise :Operation Multiply :OutputTarget Noise :InB NoiseAmplitude))
  ;;[0.15 0.15 0.15 0.5] Blur
  (bind Branch_0 (Branch :bOutputToB-param NoiseBlur))
  (bind BlurElement_1 (Blur :NumIterations 2 :NumIterations-param NoiseBlurIterations :SearchDistance-param NoiseBlurDistance))
  ;[0 0 0 0.5] Distort positions
  (bind MakeTransformAttribute_2 (Make%20Transform%20Attribute :Scale DistortScale))
  (bind Branch_1 (Branch :bOutputToB-param Distort))
  (bind TransformPoints_1 (Transform%20Points :bUniformScale false :OffsetMax-param DistortOffset :OffsetMin-param DistortOffset))
  (bind Spatial%20Noise_2 (Spatial%20Noise :bTiling true :Contrast 2.0 :RandomOffset (0.0 0.0 0.0) :Transform (struct :Scale3D (2.0 2.0 2.0)) :ValueTarget Noise :VoronoiCellRandomness 0.3608 :Iterations-param DistortIterations))
  (bind AttributeMathsOp_4 (Attribute%20Maths%20Op :Operation Multiply :InB DistortAmplitude))
  (bind AttributeMathsOp_2 (Attribute%20Maths%20Op :InputSource1 $Position.X :InputSource2 Noise :OutputTarget $Position.X))
  (bind Spatial%20Noise_3 (Spatial%20Noise :bTiling true :Contrast 2.0 :RandomOffset (0.0 0.0 0.0) :Transform (struct :Scale3D (2.0 2.0 2.0)) :ValueTarget Noise :VoronoiCellRandomness 0.3608 :Iterations-param DistortIterations))
  (bind AttributeMathsOp_5 (Attribute%20Maths%20Op :Operation Multiply :InB DistortAmplitude))
  (bind AttributeMathsOp_3 (Attribute%20Maths%20Op :InputSource1 $Position.Y :InputSource2 Noise :OutputTarget $Position.Y))
  (connect Spatial%20Noise_1 Out AttributeMathsOp_0 InA)
  (connect FilterDataByType_1 InsideFilter AttributeMathsOp_1 InA)
  (connect BlurElement_1 Out AttributeMathsOp_1 InB)
  (connect Branch_0 Output%20A AttributeMathsOp_1 InB)
  (connect Branch_1 Output%20B AttributeMathsOp_2 InA)
  (connect AttributeMathsOp_4 Out AttributeMathsOp_2 InB)
  (connect AttributeMathsOp_2 Out AttributeMathsOp_3 InA)
  (connect AttributeMathsOp_5 Out AttributeMathsOp_3 InB)
  (connect Spatial%20Noise_2 Out AttributeMathsOp_4 InA)
  (connect Spatial%20Noise_3 Out AttributeMathsOp_5 InA)
  (connect Branch_0 Output%20B BlurElement_1 In)
  (connect AttributeMathsOp_0 Out Branch_0 In)
  (connect FilterDataByType_1 InsideFilter Branch_1 In)
  (connect DefaultInputNode In FilterDataByType_1 In)
  (connect TransformPoints_0 Out Spatial%20Noise_1 In)
  (connect MakeTransformAttribute_1 Out Spatial%20Noise_1 Transform)
  (connect TransformPoints_1 Out Spatial%20Noise_2 In)
  (connect MakeTransformAttribute_2 Out Spatial%20Noise_2 Transform)
  (connect TransformPoints_1 Out Spatial%20Noise_3 In)
  (connect MakeTransformAttribute_2 Out Spatial%20Noise_3 Transform)
  (connect AttributeMathsOp_3 Out TransformPoints_0 In)
  (connect Branch_1 Output%20A TransformPoints_0 In)
  (connect Branch_1 Output%20B TransformPoints_1 In)
  (output Out AttributeMathsOp_1))
