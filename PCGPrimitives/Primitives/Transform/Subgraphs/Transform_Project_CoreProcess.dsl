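; PCG subgraph that projects primary points, splines and polygons onto secondary inputs, optionally transferring a named attribute.
; Surface targets use Projection, rebuilding splines afterwards.
; Point, spline and polygon targets become merged sample points; GPU Custom HLSL computes projection distance, driving a mask (optional falloff) that lerps Z and the attribute, optionally becoming density.
; Closed splines can become polygon targets; missing target attributes default to zero with a warning.
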(graph Transform_Project_CoreProcess
  (params (ProjectPositions :type Bool :default true) (ProjectRotations :type Bool) (ProjectScales :type Bool) (UseFalloff :type Bool) (StartFalloffDistance :type Double) (EndFalloffDistance :type Double :default 1e+03) (SetFalloffAsDensity :type Bool) (ConvertClosedSplineToPolygon2D :type Bool) (ProjectAttribute :type String))
  ;[0 0 0 0.5 12] Primary/Secondary Inputs
  (bind FilterDataByTag_0 (Filter%20Data%20By%20Tag :SelectedTags "Input:0"))
  ;[0 0 0 0.5] Surface projection branching
  (bind FilterDataByType_16 (Filter%20Data%20By%20Type :TargetType (struct :Ids [(struct :Struct /Script/PCG.PCGDataTypeInfoSurface)])))
  (bind DataNum_1 (Data%20Count))
  (bind Branch_4 (Branch))
  ;[0 0 0 0.5] Add ProjectAttribute to Source if doesnt exist
  (bind Branch_9 (Branch))
  (bind FilterDataByAttribute_0 (Filter%20Data%20By%20Attribute :Attribute-param ProjectAttribute))
  (bind AddAttribute_2 (Add%20Attribute :OutputTarget-param ProjectAttribute))
  (bind Gather_0 (Gather))
  ;[0 0 0 0.5] Project on single surface (like mesh terrain) - No falloff
  (bind MakeConcrete_3 (Make%20Concrete))
  (bind FilterDataByType_2 (Filter%20Data%20By%20Type :TargetType (struct :Ids [(struct :Struct /Script/PCG.PCGDataTypeInfoSpline)])))
  (bind GetSplineControlPoints_1 (Get%20Spline%20Control%20Points))
  (bind MakeConcrete_4 (Make%20Concrete))
  (bind FilterDataByType_10 (Filter%20Data%20By%20Type :TargetType (struct :Ids [(struct :Struct /Script/PCG.PCGDataTypeInfoPoint)])))
  (bind MakeConcrete_5 (Make%20Concrete))
  (bind FilterDataByType_3 (Filter%20Data%20By%20Type :TargetType (struct :Ids [(struct :Struct /Script/PCG.PCGDataTypeInfoPolygon2D)])))
  (bind ToPoint_0 (To%20Point))
  (bind Projection_2 (Projection :bKeepZeroDensityPoints true :ProjectionParams (struct :bProjectRotations false :AttributeMode IncludeAttributes) :AttributeList ProjectAttribute :bProjectPositions ProjectPositions :bProjectRotations ProjectRotations :bProjectScales ProjectScales))
  (bind FilterDataByTag_2 (Filter%20Data%20By%20Tag :SelectedTags "Type:Point"))
  (bind FilterDataByTag_1 (Filter%20Data%20By%20Tag :SelectedTags "Type:ClosedSpline"))
  (bind FilterDataByTag_4 (Filter%20Data%20By%20Tag :SelectedTags "Type:Polygon2D"))
  (bind ToPoint_2 (To%20Point))
  (bind CreateSpline_0 (Create%20Spline :bClosedLoop true :bLinear true :InterpTypeAttribute None))
  (bind CreateSpline_1 (Create%20Spline :bClosedLoop true :InterpTypeAttribute None))
  (bind ToPoint_3 (To%20Point))
  (bind CreateSpline_2 (Create%20Spline :InterpTypeAttribute None))
  (bind Gather_1 (Gather))
  ;[0 0 0 0.5] Copy Project Attribute
  (bind AddAttribute_0 (Add%20Attribute :OutputTarget ProjectAttribute))
  (bind Branch_0 (Branch)) ; Use ProjectAttribute
  (bind CopyAttributes_2 (Copy%20Attributes :InputSource @Last :OutputTarget ProjectAttribute :InputSource-param ProjectAttribute))
  ;[0 0 0 0.5] Splines and Polygon2D projection target
  (bind CreateConstant_1 (Create%20Constant :AttributeTypes (struct :Type Float :FloatValue 0.5)))
  (bind FilterDataByType_0 (Filter%20Data%20By%20Type :TargetType (struct :Ids [(struct :Struct /Script/PCG.PCGDataTypeInfoPolygon2D)]))) ; Polygon2D
  (bind FilterDataByType_5 (Filter%20Data%20By%20Type :TargetType (struct :Ids [(struct :Struct /Script/PCG.PCGDataTypeInfoPolygon2D)]))) ; Polygon2D
  (bind CreateSurfaceFromPolygon2D_4 (Create%20Surface%20From%20Polygon2D))
  (bind SurfaceSampler_0 (Surface%20Sampler :bApplyDensityToPoints false :bUnbounded true :Looseness 0.0 :PointExtents (0.001 0.001 0.001) :PointsPerSquaredMeter 4.0))
  (bind Branch_1 (Branch))
  (bind ExtractAttributeAtIndex_0 (Extract%20Attribute%20at%20Index :bForceOutputAttributeToBeInElementsDomain true :InputSource-param ProjectAttribute))
  (bind CopyAttributes_1 (Copy%20Attributes :InputSource @Last :InputSource-param ProjectAttribute :OutputTarget-param ProjectAttribute))
  (bind FilterDataByType_9 (Filter%20Data%20By%20Type :TargetType (struct :Ids [(struct :Struct /Script/PCG.PCGDataTypeInfoPoint)]))) ; Point
  (bind FilterDataByType_11 (Filter%20Data%20By%20Type :TargetType (struct :Ids [(struct :Struct /Script/PCG.PCGDataTypeInfoSpline)]))) ; Spline
  (bind SplineSampler_0 (Spline%20Sampler :SamplerParams (struct :Mode Distance :DistanceIncrement 5e+01 :bUnbounded true :PointSteepness 1.0)))
  (bind MergePoints_0 (Merge%20Points))
  (bind AddAttribute_1 (Add%20Attribute :OutputTarget ProjectAttribute))
  (bind Branch_5 (Branch))
  (bind CopyAttributes_3 (Copy%20Attributes :InputSource @Last :OutputTarget ProjectAttribute :InputSource-param ProjectAttribute))
  ;[0 0 0 0.5] ConvertClosedSplineToPolygon2D
  (bind Branch_11 (Branch :bOutputToB-param ConvertClosedSplineToPolygon2D))
  (bind FilterDataByTag_3 (Filter%20Data%20By%20Tag :SelectedTags "Type:ClosedSpline"))
  (bind FilterDataByType_4 (Filter%20Data%20By%20Type :TargetType (struct :Ids [(struct :Struct /Script/PCG.PCGDataTypeInfoPoint) (struct :Struct /Script/PCG.PCGDataTypeInfoSpline)])))
  (bind CreatePolygon2D_2 (Create%20Polygon%202D))
  (bind Gather_3 (Gather))
  ;[0 0 0 0.5] Validating if Project Attribute exist on secondary input data, if not add it (default 0) and log a warning to indicate it is missing
  (bind Branch_10 (Branch))
  (bind FilterDataByType_6 (Filter%20Data%20By%20Type :bShowOutsideFilter true :TargetType (struct :Ids [(struct :Struct /Script/PCG.PCGDataTypeInfoSurface)]))) ; Removing surface handled in its own branch
  (bind WriteDataIndex_1 (Write%20Data%20Index :IndexTag "DataIndex"))
  (bind FilterDataByAttribute_3 (Filter%20Data%20By%20Attribute :Attribute-param ProjectAttribute))
  (bind AddAttribute_5 (Add%20Attribute :OutputTarget-param ProjectAttribute))
  (bind SortTags_1 (Sort%20Data%20By%20Tag%20Value :Tag DataIndex))
  (bind DeleteTags_0 (Delete%20Tags :SelectedTags "DataIndex"))
  ;;[1 0.943 0 0.5] Missing Attribute
  (bind Print%20String_0 (Print%20String :bPrefixWithComponent true :bPrefixWithNode false :bPrefixWithOwner true :PrintString "The Project Attribute set in the primitive does not exist on the provided secondary input data and has been defaulted to 0." :Verbosity Warning))
  ;[0 0 0 0.5] Execute branch
  (bind DataNum_0 (Data%20Count))
  (bind Branch_2 (Branch))
  (bind Branch_3 (Branch))
  ;[0 0 0 0.5] Project points on splines and polygon2d with optional falloff
  (bind FilterDataByType_1 (Filter%20Data%20By%20Type :TargetType (struct :Ids [(struct :Struct /Script/PCG.PCGDataTypeInfoPoint)])))
  (bind CustomHLSL_1 (Custom%20HLSL :InputPins [(struct :Label In1 :AllowedTypes (struct :Ids [(struct :Struct /Script/PCG.PCGDataTypeInfoPoint)]) :PinStatus Required :bAllowMultipleConnections false) (struct :Label In2 :AllowedTypes (struct :Ids [(struct :Struct /Script/PCG.PCGDataTypeInfoPoint)]) :bAllowMultipleConnections false)] :OutputPins [(struct :PropertiesGPU (struct :PinsToInititalizeFrom [In1] :CreatedKernelAttributeKeys [(struct :Identifier (struct :Name ProjectPosition) :Type Float3 :Name ProjectPosition) (struct :Identifier (struct :Name ProjectRotation) :Type Float3 :Name ProjectRotation) (struct :Identifier (struct :Name ProjectScale) :Type Float3 :Name ProjectScale) (struct :Identifier (struct :Name ProjectDistance) :Name ProjectDistance)]) :Label Out :AllowedTypes (struct :Ids [(struct :Struct /Script/PCG.PCGDataTypeInfoPoint)]))]))
  (bind AttributeRemap_4 (Attribute%20Remap :bClampToUnitRange true :InputSource ProjectDistance :OutputTarget ProjectMask :OutRangeMax 0.0 :OutRangeMin 1.0 :InRangeMax-param EndFalloffDistance :InRangeMin-param StartFalloffDistance))
  (bind AttributeRemap_3 (Attribute%20Remap :Curve (struct :EditorCurveData (struct :Keys [(struct :InterpMode RCIM_Cubic) (struct :Time 1.0 :Value 1.0)])) :InputSource ProjectMask :Mode Curve :OutputTarget ProjectMask))
  (bind AttributeRemap_0 (Attribute%20Remap :bClampToUnitRange true :InputSource ProjectDistance :InRangeMax 1e+02 :InRangeMin 99.0 :OutputTarget ProjectMask :OutRangeMax 0.0 :OutRangeMin 1.0))
  (bind Select_0 (Select :bUseInputB-param UseFalloff))
  (bind DeleteAttributes_1 (Delete%20Attributes :Operator Matches :SelectedAttributes "Project*"))
  ;;[0 0 0 0.5] Use Project Position
  (bind Switch_9 (Switch :EnumSelection (struct :Class /Script/ACLPlugin.ACLCompressionLevel) :IntegerSelection-param ProjectPositions)) ; Use ProjectPosition
  (bind AttributeMathsOp_1 (Attribute%20Maths%20Op :InputSource1 $Position.Z :InputSource2 ProjectPosition.Z :InputSource3 ProjectMask :Operation Lerp :OutputTarget $Position.Z)) ; ProjectPosition
  ;;[0 0 0 0.5] Use ProjectAttribute
  (bind Branch_6 (Branch))
  (bind AttributeMathsOp_8 (Attribute%20Maths%20Op :InputSource1 $Position.Z :InputSource2 ProjectAttribute :InputSource3 ProjectMask :Operation Lerp :OutputTarget $Position.Z :InputSource1-param ProjectAttribute :OutputTarget-param ProjectAttribute)) ; ProjectAttribute
  ;;[0 0 0 0.5] Set Falloff as Density
  (bind AttributeMathsOp_2 (Attribute%20Maths%20Op :InputSource1 ProjectMask :InputSource2 ProjectMask :Operation Set :OutputTarget $Density)) ; Debug Mask  With Density
  (bind Select_1 (Select :bUseInputB-param SetFalloffAsDensity))
  ;[0 0 0 0.5] Splines on splines with optional falloff
  (bind FilterDataByType_12 (Filter%20Data%20By%20Type :TargetType (struct :Ids [(struct :Struct /Script/PCG.PCGDataTypeInfoSpline)])))
  (bind GetSplineControlPoints_0 (Get%20Spline%20Control%20Points))
  (bind CustomHLSL_3 (Custom%20HLSL :InputPins [(struct :Label In1 :AllowedTypes (struct :Ids [(struct :Struct /Script/PCG.PCGDataTypeInfoPoint)]) :PinStatus Required :bAllowMultipleConnections false) (struct :Label In2 :AllowedTypes (struct :Ids [(struct :Struct /Script/PCG.PCGDataTypeInfoPoint)]) :bAllowMultipleConnections false)] :OutputPins [(struct :PropertiesGPU (struct :PinsToInititalizeFrom [In1] :CreatedKernelAttributeKeys [(struct :Identifier (struct :Name ProjectPosition) :Type Float3 :Name ProjectPosition) (struct :Identifier (struct :Name ProjectRotation) :Type Float3 :Name ProjectRotation) (struct :Identifier (struct :Name ProjectScale) :Type Float3 :Name ProjectScale) (struct :Identifier (struct :Name ProjectDistance) :Name ProjectDistance)]) :Label Out :AllowedTypes (struct :Ids [(struct :Struct /Script/PCG.PCGDataTypeInfoPoint)]))]))
  (bind AttributeRemap_2 (Attribute%20Remap :bClampToUnitRange true :InputSource ProjectDistance :OutputTarget ProjectMask :OutRangeMax 0.0 :OutRangeMin 1.0 :InRangeMax-param EndFalloffDistance :InRangeMin-param StartFalloffDistance))
  (bind AttributeRemap_1 (Attribute%20Remap :Curve (struct :EditorCurveData (struct :Keys [(struct :InterpMode RCIM_Cubic) (struct :Time 1.0 :Value 1.0)])) :InputSource ProjectMask :Mode Curve :OutputTarget ProjectMask))
  (bind AttributeRemap_5 (Attribute%20Remap :bClampToUnitRange true :InputSource ProjectDistance :InRangeMax 1e+02 :InRangeMin 99.0 :OutputTarget ProjectMask :OutRangeMax 0.0 :OutRangeMin 1.0))
  (bind Select_2 (Select :bUseInputB-param UseFalloff))
  (bind CopyAttributes_0 (Copy%20Attributes :InputSource $Position :OutputTarget $Position))
  (bind DeleteAttributes_0 (Delete%20Attributes :Operator Matches :SelectedAttributes "Project*"))
  (bind FilterDataByTag_5 (Filter%20Data%20By%20Tag :Operation RemoveTagged :SelectedTags "Type:Polygon2d"))
  (bind CreatePolygon2D_0 (Create%20Polygon%202D :InputType ForceClosed)) ; Recreate input polygons
  ;;[0 0 0 0.5] Use Project Position
  (bind Switch_8 (Switch :EnumSelection (struct :Class /Script/ACLPlugin.ACLCompressionLevel) :IntegerSelection-param ProjectPositions)) ; Use ProjectPosition
  (bind AttributeMathsOp_0 (Attribute%20Maths%20Op :InputSource1 $Position.Z :InputSource2 ProjectPosition.Z :InputSource3 ProjectMask :Operation Lerp :OutputTarget $Position.Z)) ; ProjectPosition
  ;;[0 0 0 0.5] Use ProjectAttribute
  (bind Branch_7 (Branch))
  (bind AttributeMathsOp_7 (Attribute%20Maths%20Op :InputSource1 $Position.Z :InputSource2 ProjectAttribute :InputSource3 ProjectMask :Operation Lerp :OutputTarget $Position.Z :InputSource1-param ProjectAttribute :OutputTarget-param ProjectAttribute)) ; ProjectAttribute
  ;;[0 0 0 0.5] Set Falloff as Density
  (bind AttributeMathsOp_3 (Attribute%20Maths%20Op :InputSource1 ProjectMask :InputSource2 ProjectMask :Operation Set :OutputTarget $Density)) ; Debug Mask  With Density
  (bind Select_3 (Select :bUseInputB-param SetFalloffAsDensity))
  ;;[0 0 0 0.5] Copy ProjectAttribute to original spline
  (bind Branch_8 (Branch))
  (bind CopyAttributes_4 (Copy%20Attributes :InputSource $Position :OutputTarget $Position :InputSource-param ProjectAttribute :OutputTarget-param ProjectAttribute))
  ;[0 0 0 0.5] Use ProjectAttribute true/false
  (bind AttributeCompareOp_0 (Attribute%20Compare%20Op :Operation NotEqual :InA ProjectAttribute))
  (reroute UseProjectAttributeInt AttributeCompareOp_0)
  (connect Gather_1 Out AddAttribute_0 In)
  (connect MergePoints_0 Out AddAttribute_1 In)
  (connect FilterDataByAttribute_0 OutsideFilter AddAttribute_2 In)
  (connect FilterDataByAttribute_3 OutsideFilter AddAttribute_5 In)
  (connect Switch_8 Default AttributeMathsOp_0 InA)
  (connect Switch_8 Default AttributeMathsOp_0 InB)
  (connect Switch_8 Default AttributeMathsOp_0 Ratio)
  (connect Switch_9 Default AttributeMathsOp_1 InA)
  (connect Switch_9 Default AttributeMathsOp_1 InB)
  (connect Switch_9 Default AttributeMathsOp_1 Ratio)
  (connect AttributeMathsOp_8 Out AttributeMathsOp_2 InA)
  (connect Branch_6 Output%20A AttributeMathsOp_2 InA)
  (connect AttributeMathsOp_8 Out AttributeMathsOp_2 InB)
  (connect Branch_6 Output%20A AttributeMathsOp_2 InB)
  (connect AttributeMathsOp_7 Out AttributeMathsOp_3 InA)
  (connect Branch_7 Output%20A AttributeMathsOp_3 InA)
  (connect AttributeMathsOp_7 Out AttributeMathsOp_3 InB)
  (connect Branch_7 Output%20A AttributeMathsOp_3 InB)
  (connect Branch_7 Output%20B AttributeMathsOp_7 InA)
  (connect Branch_7 Output%20B AttributeMathsOp_7 InB)
  (connect Branch_7 Output%20B AttributeMathsOp_7 Ratio)
  (connect Branch_6 Output%20B AttributeMathsOp_8 InA)
  (connect Branch_6 Output%20B AttributeMathsOp_8 InB)
  (connect Branch_6 Output%20B AttributeMathsOp_8 Ratio)
  (connect CustomHLSL_1 Out AttributeRemap_0 In)
  (connect AttributeRemap_2 Out AttributeRemap_1 In)
  (connect CustomHLSL_3 Out AttributeRemap_2 In)
  (connect AttributeRemap_4 Out AttributeRemap_3 In)
  (connect CustomHLSL_1 Out AttributeRemap_4 In)
  (connect CustomHLSL_3 Out AttributeRemap_5 In)
  (connect UseProjectAttributeInt Out Branch_0 bOutputToB)
  (connect AddAttribute_0 Out Branch_0 In)
  (connect UseProjectAttributeInt Out Branch_1 bOutputToB)
  (connect SurfaceSampler_0 Out Branch_1 In)
  (connect DataNum_0 Out Branch_2 bOutputToB)
  (connect Branch_0 Output%20A Branch_2 In)
  (connect CopyAttributes_2 Out Branch_2 In)
  (connect DataNum_0 Out Branch_3 bOutputToB)
  (connect Branch_5 Output%20A Branch_3 In)
  (connect CopyAttributes_3 Out Branch_3 In)
  (connect DataNum_1 Out Branch_4 bOutputToB)
  (connect Gather_0 Out Branch_4 In)
  (connect UseProjectAttributeInt Out Branch_5 bOutputToB)
  (connect AddAttribute_1 Out Branch_5 In)
  (connect UseProjectAttributeInt Out Branch_6 bOutputToB)
  (connect AttributeMathsOp_1 Out Branch_6 In)
  (connect Switch_9 %30 Branch_6 In)
  (connect UseProjectAttributeInt Out Branch_7 bOutputToB)
  (connect AttributeMathsOp_0 Out Branch_7 In)
  (connect Switch_8 %30 Branch_7 In)
  (connect UseProjectAttributeInt Out Branch_8 bOutputToB)
  (connect CopyAttributes_0 Out Branch_8 In)
  (connect UseProjectAttributeInt Out Branch_9 bOutputToB)
  (connect FilterDataByTag_0 InsideFilter Branch_9 In)
  (connect UseProjectAttributeInt Out Branch_10 bOutputToB)
  (connect Branch_11 Output%20A Branch_10 In)
  (connect Gather_3 Out Branch_10 In)
  (connect FilterDataByTag_0 OutsideFilter Branch_11 In)
  (connect Select_3 Out CopyAttributes_0 Source)
  (connect FilterDataByType_12 InsideFilter CopyAttributes_0 Target)
  (connect ExtractAttributeAtIndex_0 Out CopyAttributes_1 Source)
  (connect Branch_1 Output%20B CopyAttributes_1 Target)
  (connect Branch_0 Output%20B CopyAttributes_2 Source)
  (connect Branch_0 Output%20B CopyAttributes_2 Target)
  (connect Branch_5 Output%20B CopyAttributes_3 Source)
  (connect Branch_5 Output%20B CopyAttributes_3 Target)
  (connect Select_3 Out CopyAttributes_4 Source)
  (connect Branch_8 Output%20B CopyAttributes_4 Target)
  (connect FilterDataByTag_5 OutsideFilter CreatePolygon2D_0 In)
  (connect FilterDataByType_4 InsideFilter CreatePolygon2D_2 In)
  (connect ToPoint_2 Out CreateSpline_0 In)
  (connect FilterDataByTag_1 InsideFilter CreateSpline_1 In)
  (connect ToPoint_3 Out CreateSpline_2 In)
  (connect FilterDataByType_0 InsideFilter CreateSurfaceFromPolygon2D_4 In)
  (connect FilterDataByType_5 InsideFilter CreateSurfaceFromPolygon2D_4 In)
  (connect FilterDataByType_1 InsideFilter CustomHLSL_1 In1)
  (connect Branch_3 Output%20B CustomHLSL_1 In2)
  (connect GetSplineControlPoints_0 Out CustomHLSL_3 In1)
  (connect Branch_3 Output%20B CustomHLSL_3 In2)
  (connect Branch_5 Output%20A DataNum_0 In)
  (connect CopyAttributes_3 Out DataNum_0 In)
  (connect FilterDataByType_16 InsideFilter DataNum_1 In)
  (connect Branch_8 Output%20A DeleteAttributes_0 In)
  (connect CopyAttributes_4 Out DeleteAttributes_0 In)
  (connect Select_1 Out DeleteAttributes_1 In)
  (connect SortTags_1 Out DeleteTags_0 In)
  (connect FilterDataByType_5 InsideFilter ExtractAttributeAtIndex_0 In)
  (connect Branch_9 Output%20B FilterDataByAttribute_0 In)
  (connect WriteDataIndex_1 Out FilterDataByAttribute_3 In)
  (connect DefaultInputNode In FilterDataByTag_0 In)
  (connect FilterDataByTag_2 OutsideFilter FilterDataByTag_1 In)
  (connect Projection_2 Out FilterDataByTag_2 In)
  (connect Branch_11 Output%20B FilterDataByTag_3 In)
  (connect FilterDataByTag_1 OutsideFilter FilterDataByTag_4 In)
  (connect DeleteAttributes_0 Out FilterDataByTag_5 In)
  (connect Branch_10 Output%20A FilterDataByType_0 In)
  (connect Branch_2 Output%20B FilterDataByType_1 In)
  (connect Branch_4 Output%20B FilterDataByType_2 In)
  (connect Branch_4 Output%20B FilterDataByType_3 In)
  (connect FilterDataByTag_3 InsideFilter FilterDataByType_4 In)
  (connect DeleteTags_0 Out FilterDataByType_5 In)
  (connect Branch_10 Output%20B FilterDataByType_6 In)
  (connect Branch_10 Output%20A FilterDataByType_9 In)
  (connect DeleteTags_0 Out FilterDataByType_9 In)
  (connect Branch_4 Output%20B FilterDataByType_10 In)
  (connect Branch_10 Output%20A FilterDataByType_11 In)
  (connect DeleteTags_0 Out FilterDataByType_11 In)
  (connect Branch_2 Output%20B FilterDataByType_12 In)
  (connect FilterDataByTag_0 OutsideFilter FilterDataByType_16 In)
  (connect Branch_9 Output%20A Gather_0 In)
  (connect FilterDataByAttribute_0 InsideFilter Gather_0 In)
  (connect AddAttribute_2 Out Gather_0 In2)
  (connect Branch_4 Output%20A Gather_1 In)
  (connect CreateSpline_0 Out Gather_1 In)
  (connect CreateSpline_1 Out Gather_1 In)
  (connect CreateSpline_2 Out Gather_1 In)
  (connect FilterDataByTag_2 InsideFilter Gather_1 In)
  (connect CreatePolygon2D_2 Out Gather_3 In)
  (connect FilterDataByTag_3 OutsideFilter Gather_3 In)
  (connect FilterDataByType_12 InsideFilter GetSplineControlPoints_0 In)
  (connect FilterDataByType_2 InsideFilter GetSplineControlPoints_1 In)
  (connect FilterDataByType_16 InsideFilter MakeConcrete_3 In)
  (connect GetSplineControlPoints_1 Out MakeConcrete_4 In)
  (connect FilterDataByType_10 InsideFilter MakeConcrete_5 In)
  (connect Branch_1 Output%20A MergePoints_0 In)
  (connect CopyAttributes_1 Out MergePoints_0 In)
  (connect FilterDataByType_9 InsideFilter MergePoints_0 In)
  (connect SplineSampler_0 Out MergePoints_0 In)
  (connect AddAttribute_5 Out Print%20String_0 In)
  (connect MakeConcrete_4 Out Projection_2 In)
  (connect MakeConcrete_5 Out Projection_2 In)
  (connect ToPoint_0 Out Projection_2 In)
  (connect MakeConcrete_3 Out Projection_2 Projection%20Target)
  (connect AttributeRemap_0 Out Select_0 Input%20A)
  (connect AttributeRemap_3 Out Select_0 Input%20B)
  (connect AttributeMathsOp_8 Out Select_1 Input%20A)
  (connect Branch_6 Output%20A Select_1 Input%20A)
  (connect AttributeMathsOp_2 Out Select_1 Input%20B)
  (connect AttributeRemap_5 Out Select_2 Input%20A)
  (connect AttributeRemap_1 Out Select_2 Input%20B)
  (connect AttributeMathsOp_7 Out Select_3 Input%20A)
  (connect Branch_7 Output%20A Select_3 Input%20A)
  (connect AttributeMathsOp_3 Out Select_3 Input%20B)
  (connect AddAttribute_5 Out SortTags_1 In)
  (connect FilterDataByAttribute_3 InsideFilter SortTags_1 In)
  (connect FilterDataByType_11 InsideFilter SplineSampler_0 Spline)
  (connect CreateConstant_1 Out SurfaceSampler_0 PointsPerSquaredMeter)
  (connect CreateSurfaceFromPolygon2D_4 Out SurfaceSampler_0 Surface)
  (connect Select_2 Out Switch_8 In)
  (connect Select_0 Out Switch_9 In)
  (connect FilterDataByType_3 InsideFilter ToPoint_0 In)
  (connect FilterDataByTag_4 InsideFilter ToPoint_2 In)
  (connect FilterDataByTag_4 OutsideFilter ToPoint_3 In)
  (connect FilterDataByType_6 OutsideFilter WriteDataIndex_1 In)
  (output Out Branch_2.Output%20A)
  (output Out CreatePolygon2D_0)
  (output Out DeleteAttributes_1)
  (output Out FilterDataByTag_5.InsideFilter))
