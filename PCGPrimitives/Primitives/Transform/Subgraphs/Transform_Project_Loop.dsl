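; PCG loop subgraph for the Transform Project primitive that prepares each looped data item using an optional ProjectAttribute name.
; A string filter on that name drives a switch that either skips the attribute handling or checks whether the data has the attribute.
; A second switch, driven by the matching data count, adds the attribute when missing.
; All branches converge on an attribute maths op reading Position.Z, which feeds the output.
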(graph Transform_Project_Loop
  (params (ProjectAttribute :type String))
  (inputs (Loop :type Any))
  ;[0.15 0.15 0.15 0.5] Add ProjectAttribute to Source if doesnt exist
  (bind AttributeSetToPoint_1 (Attribute%20Set%20To%20Point :In ProjectAttribute))
  (bind AttributeFilter_1 (Filter%20Attribute%20Elements :AttributeTypes (struct :Type String) :bUseConstantThreshold true :Operator Matches :TargetAttribute ProjectAttribute :ThresholdAttribute @Last))
  (bind GetElementsCount_1 (Get%20Element%20Count))
  (bind Switch_0 (Switch :EnumSelection (struct :Class /Script/ACLPlugin.ACLCompressionLevel))) ; Use ProjectAttribute
  (bind FilterDataByAttribute_1 (Filter%20Data%20By%20Attribute :Attribute-param ProjectAttribute))
  (bind GetElementsCount_0 (Get%20Element%20Count))
  (bind DataNum_0 (Data%20Count))
  (bind Switch_1 (Switch :EnumSelection (struct :Class /Script/ACLPlugin.ACLCompressionLevel))) ; Add Missing ProjectAttribute
  (bind AddAttribute_0 (Add%20Attribute :OutputTarget-param ProjectAttribute))
  (bind AttributeMathsOp_0 (Attribute%20Maths%20Op :InputSource1 $Position.Z))
  (connect Switch_1 %30 AddAttribute_0 In)
  (connect AttributeSetToPoint_1 Out AttributeFilter_1 In)
  (connect AddAttribute_0 Out AttributeMathsOp_0 InA)
  (connect Switch_0 Default AttributeMathsOp_0 InA)
  (connect Switch_1 Default AttributeMathsOp_0 InA)
  (connect FilterDataByAttribute_1 InsideFilter DataNum_0 In)
  (connect Switch_0 %30 FilterDataByAttribute_1 In)
  (connect FilterDataByAttribute_1 InsideFilter GetElementsCount_0 In)
  (connect AttributeFilter_1 InsideFilter GetElementsCount_1 In)
  (connect DefaultInputNode Loop Switch_0 In)
  (connect GetElementsCount_1 Out Switch_0 IntegerSelection)
  (connect Switch_0 %30 Switch_1 In)
  (connect DataNum_0 Out Switch_1 IntegerSelection)
  (output Out AttributeMathsOp_0))
