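; PCG subgraph that resamples splines by sampling them into points and rebuilding new splines, as the core of a Resample transform primitive.
; Only spline input reaches the Spline Sampler, whose mode, subdivision, distance and fit-to-curve come from graph parameters.
; The "Type:Spline" tag routes samples to open-spline creation; all others become closed loops.
; Both branches replace old Type tags with "Type:Spline" or "Type:ClosedSpline" and merge at Out.
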(graph Transform_Resample_CoreProcess
  (params (SamplingMode :type (Enum /PCGPrimitives/Primitives/Transform/Subgraphs/EPCGSplineSamplingMode.EPCGSplineSamplingMode)) (Subdivision :type Int :default 1) (Distance :type Double :default 5e+02) (FitToCurve :type Bool))
  (bind FilterDataByType_0 (Filter%20Data%20By%20Type :TargetType (struct :Ids [(struct :Struct /Script/PCG.PCGDataTypeInfoSpline)])))
  (bind SplineSampler_1 (Spline%20Sampler :SamplerParams (struct :Mode Distance :bUnbounded true :PointSteepness 1.0) :bFitToCurve FitToCurve :DistanceIncrement Distance :Mode SamplingMode :NumSamples Subdivision :SubdivisionsPerSegment Subdivision))
  (bind FilterDataByTag_1 (Filter%20Data%20By%20Tag :SelectedTags "Type:Spline"))
  (bind CreateSpline_1 (Create%20Spline :bClosedLoop true :InterpTypeAttribute None))
  (bind CreateSpline_0 (Create%20Spline :InterpTypeAttribute None))
  (bind DeleteTags_1 (Delete%20Tags :Operator Substring :SelectedTags "Type:Point, Type:Spline, Type:ClosedSpline, Type:Polygon2D, Type:Surface, Type:Texture2D, Type:Volume"))
  (bind AddTags_1 (Add%20Tags :TagsToAdd "Type:ClosedSpline"))
  (bind DeleteTags_2 (Delete%20Tags :Operator Substring :SelectedTags "Type:Point, Type:Spline, Type:ClosedSpline, Type:Polygon2D, Type:Surface, Type:Texture2D, Type:Volume"))
  (bind AddTags_2 (Add%20Tags :TagsToAdd "Type:Spline"))
  (connect DeleteTags_1 Out AddTags_1 In)
  (connect DeleteTags_2 Out AddTags_2 In)
  (connect FilterDataByTag_1 InsideFilter CreateSpline_0 In)
  (connect FilterDataByTag_1 OutsideFilter CreateSpline_1 In)
  (connect CreateSpline_1 Out DeleteTags_1 In)
  (connect CreateSpline_0 Out DeleteTags_2 In)
  (connect SplineSampler_1 Out FilterDataByTag_1 In)
  (connect DefaultInputNode In FilterDataByType_0 In)
  (connect FilterDataByType_0 InsideFilter SplineSampler_1 Spline)
  (output Out AddTags_1)
  (output Out AddTags_2))
