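; PCG subgraph that randomly disturbs the inner control points of open input splines in X and Y, using Disturbance_amount and seed.
; It tags control points with their data index, separates first and last by element count, and offsets only the inner ones.
; Merged points are sorted and partitioned per spline, then their positions are copied onto the original splines.
; Splines are rebuilt from those control points with custom tangents and interpolation types.
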(graph Transform_Splines_Disturb_CoreProcess
  (params (Disturbance_amount :type Double) (seed :type Int))
  (bind FilterDataByType_0 (Filter%20Data%20By%20Type :TargetType (struct :Ids [(struct :Struct /Script/PCG.PCGDataTypeInfoSpline)])))
  (bind FilterDataByAttribute_0 (Filter%20Data%20By%20Attribute :FilterMode FilterByValue :FilterOperator Equal :TargetAttribute @Data.$isClosed :Threshold (struct :AttributeTypes (struct :Type Boolean))))
  (bind GetSplineControlPoints_1 (Get%20Spline%20Control%20Points))
  (bind WriteDataIndex_0 (Write%20Data%20Index :bAddAttribute true :bAddTag false :IndexAttribute dataIndex))
  ;[0 0 0 0.5] Isolate start and end from inner spline control points.
  (bind GetElementsCount_0 (Get%20Element%20Count))
  (bind AttributeMathsOp_2 (Attribute%20Maths%20Op :Operation Subtract))
  (bind AttributeFilterRange_1 (Filter%20Attribute%20Elements%20by%20Range :MaxThreshold (struct :bInclusive false :ThresholdAttribute @Last :bUseSpatialQuery false) :MinThreshold (struct :bInclusive false :bUseConstantThreshold true :ThresholdAttribute @Last :bUseSpatialQuery false :AttributeTypes (struct :Type Integer32)) :TargetAttribute $Index))
  (bind AttributeFilter_0 (Filter%20Attribute%20Elements :AttributeTypes (struct :Type Integer32) :bUseConstantThreshold true :Operator Equal :TargetAttribute $Index :ThresholdAttribute @Last))
  (bind AttributeFilter_1 (Filter%20Attribute%20Elements :Operator Equal :TargetAttribute $Index :ThresholdAttribute @Last))
  ;[0 0 0 0.5] Disturb inner spline control points in X,Y only
  (bind MakeVectorAttribute_0 (Make%20Vector%20Attribute :OutputType Vector :X Disturbance_amount :Y Disturbance_amount))
  (bind AttributeMathsOp_1 (Attribute%20Maths%20Op :Operation Multiply))
  (bind TransformPoints_1 (Transform%20Points :Seed-param seed))
  ;[0 0 0 0.5] Merge and partition according to input data index
  (bind MergePoints_0 (Merge%20Points))
  (bind SortAttributes_0 (Sort%20Attributes :bUseStableSort false :SortAttributes [(struct :InputSource dataindex)]))
  (bind AttributePartition_0 (Attribute%20Partition :PartitionAttributeSelectors [dataIndex]))
  ;[0 0 0 0.5] Apply modified positions onto original input spline control points
  (bind CopyAttributes_1 (Copy%20Attributes :InputSource $Position :OutputTarget $Position))
  ;[0 0 0 0.5] Recompute splines
  (bind GetSplineControlPoints_2 (Get%20Spline%20Control%20Points))
  (bind CreateSpline_0 (Create%20Spline :bApplyCustomTangents true :bUseInterpTypeAttribute true))
  (connect WriteDataIndex_0 Out AttributeFilter_0 In)
  (connect AttributeMathsOp_2 Out AttributeFilter_1 Filter)
  (connect WriteDataIndex_0 Out AttributeFilter_1 In)
  (connect AttributeMathsOp_2 Out AttributeFilterRange_1 FilterMax)
  (connect WriteDataIndex_0 Out AttributeFilterRange_1 In)
  (connect MakeVectorAttribute_0 Out AttributeMathsOp_1 InA)
  (connect GetElementsCount_0 Out AttributeMathsOp_2 InA)
  (connect SortAttributes_0 Out AttributePartition_0 In)
  (connect AttributePartition_0 Out CopyAttributes_1 Source)
  (connect FilterDataByAttribute_0 InsideFilter CopyAttributes_1 Target)
  (connect GetSplineControlPoints_2 Out CreateSpline_0 In)
  (connect FilterDataByType_0 InsideFilter FilterDataByAttribute_0 In)
  (connect DefaultInputNode In FilterDataByType_0 In)
  (connect WriteDataIndex_0 Out GetElementsCount_0 In)
  (connect FilterDataByAttribute_0 InsideFilter GetSplineControlPoints_1 In)
  (connect CopyAttributes_1 Out GetSplineControlPoints_2 In)
  (connect AttributeFilter_0 InsideFilter MergePoints_0 In)
  (connect AttributeFilter_1 InsideFilter MergePoints_0 In)
  (connect TransformPoints_1 Out MergePoints_0 In)
  (connect MergePoints_0 Out SortAttributes_0 In)
  (connect AttributeFilterRange_1 InsideFilter TransformPoints_1 In)
  (connect MakeVectorAttribute_0 Out TransformPoints_1 OffsetMax)
  (connect AttributeMathsOp_1 Out TransformPoints_1 OffsetMin)
  (connect GetSplineControlPoints_1 Out WriteDataIndex_0 In)
  (output Out CreateSpline_0))
