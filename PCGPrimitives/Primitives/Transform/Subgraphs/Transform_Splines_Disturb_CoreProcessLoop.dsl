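; PCG subgraph that randomly disturbs spline control points in X and Y by Disturbance_amount, as the core loop of a spline disturb primitive.
; Control points are split by index: interior points get randomly offset, while first and last stay put, then all merge.
; Positions are copied back onto the original splines, which are rebuilt as open or closed via the isClosed attribute.
; The seed parameter is declared but not visibly used.
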(graph Transform_Splines_Disturb_CoreProcessLoop
  (params (Disturbance_amount :type Double) (seed :type Int))
  (bind FilterDataByType_0 (Filter%20Data%20By%20Type :TargetType (struct :Ids [(struct :Struct /Script/PCG.PCGDataTypeInfoPolyline)])))
  (bind GetSplineControlPoints_0 (Get%20Spline%20Control%20Points))
  (bind WriteDataIndex_1 (Write%20Data%20Index :bAddAttribute true :bAddTag false :IndexAttribute dataIndex))
  (bind GetElementsCount_1 (Get%20Element%20Count))
  (bind AttributeMathsOp_1 (Attribute%20Maths%20Op :Operation Subtract))
  (bind AttributeFilterRange_0 (Filter%20Attribute%20Elements%20by%20Range :MaxThreshold (struct :bInclusive false :ThresholdAttribute @Last :bUseSpatialQuery false) :MinThreshold (struct :bInclusive false :bUseConstantThreshold true :ThresholdAttribute @Last :bUseSpatialQuery false :AttributeTypes (struct :Type Integer32)) :TargetAttribute $Index))
  (bind MakeVectorAttribute_1 (Make%20Vector%20Attribute :OutputType Vector :X Disturbance_amount :Y Disturbance_amount))
  (bind AttributeMathsOp_0 (Attribute%20Maths%20Op :Operation Multiply))
  (bind TransformPoints_0 (Transform%20Points))
  (bind AttributeFilter_1 (Filter%20Attribute%20Elements :AttributeTypes (struct :Type Integer32) :bUseConstantThreshold true :Operator Equal :TargetAttribute $Index :ThresholdAttribute @Last))
  (bind AttributeFilter_2 (Filter%20Attribute%20Elements :Operator Equal :TargetAttribute $Index :ThresholdAttribute @Last))
  (bind MergePoints_1 (Merge%20Points))
  (bind AttributePartition_1 (Attribute%20Partition :PartitionAttributeSelectors [dataIndex]))
  (bind CopyAttributes_0 (Copy%20Attributes :InputSource $Position :OutputTarget $Position))
  (bind FilterDataByAttribute_0 (Filter%20Data%20By%20Attribute :FilterMode FilterByValue :FilterOperator Equal :TargetAttribute @Data.$isClosed :Threshold (struct :AttributeTypes (struct :Type Boolean :BoolValue true))))
  (bind GetSplineControlPoints_1 (Get%20Spline%20Control%20Points))
  (bind CreateSpline_1 (Create%20Spline :InterpTypeAttribute None))
  (bind GetSplineControlPoints_2 (Get%20Spline%20Control%20Points))
  (bind CreateSpline_2 (Create%20Spline :bApplyCustomTangents true :bClosedLoop true :InterpTypeAttribute None))
  (connect WriteDataIndex_1 Out AttributeFilter_1 In)
  (connect AttributeMathsOp_1 Out AttributeFilter_2 Filter)
  (connect WriteDataIndex_1 Out AttributeFilter_2 In)
  (connect AttributeMathsOp_1 Out AttributeFilterRange_0 FilterMax)
  (connect WriteDataIndex_1 Out AttributeFilterRange_0 In)
  (connect MakeVectorAttribute_1 Out AttributeMathsOp_0 InA)
  (connect GetElementsCount_1 Out AttributeMathsOp_1 InA)
  (connect MergePoints_1 Out AttributePartition_1 In)
  (connect AttributePartition_1 Out CopyAttributes_0 Source)
  (connect FilterDataByType_0 InsideFilter CopyAttributes_0 Target)
  (connect GetSplineControlPoints_1 Out CreateSpline_1 In)
  (connect GetSplineControlPoints_2 Out CreateSpline_2 In)
  (connect CopyAttributes_0 Out FilterDataByAttribute_0 In)
  (connect DefaultInputNode In FilterDataByType_0 In)
  (connect WriteDataIndex_1 Out GetElementsCount_1 In)
  (connect FilterDataByType_0 InsideFilter GetSplineControlPoints_0 In)
  (connect FilterDataByAttribute_0 OutsideFilter GetSplineControlPoints_1 In)
  (connect FilterDataByAttribute_0 InsideFilter GetSplineControlPoints_2 In)
  (connect AttributeFilter_1 InsideFilter MergePoints_1 In)
  (connect AttributeFilter_2 InsideFilter MergePoints_1 In)
  (connect TransformPoints_0 Out MergePoints_1 In)
  (connect AttributeFilterRange_0 InsideFilter TransformPoints_0 In)
  (connect MakeVectorAttribute_1 Out TransformPoints_0 OffsetMax)
  (connect AttributeMathsOp_0 Out TransformPoints_0 OffsetMin)
  (connect GetSplineControlPoints_0 Out WriteDataIndex_1 In)
  (output Out CreateSpline_1)
  (output Out CreateSpline_2))
