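; PCG subgraph that serves as the core process for the Transform Splines Extend primitive, exposing a Distance parameter.
; It filters the input down to spline data, then runs the Transform_Splines_Extend_CoreProcessLoop subgraph in a Loop node, passing Distance through.
; The loop result goes to the output; non-spline data is dropped, and the extension logic itself lives in the loop subgraph.
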(graph Transform_Splines_Extend_CoreProcess
  (params (Distance :type Double))
  (bind FilterDataByType_0 (Filter%20Data%20By%20Type :TargetType (struct :Ids [(struct :Struct /Script/PCG.PCGDataTypeInfoSpline)] :CustomSubtype 0)))
  (bind Loop_0 (Loop :Graph /PCGPrimitives/Primitives/Transform/Subgraphs/Transform_Splines_Extend_CoreProcessLoop.Transform_Splines_Extend_CoreProcessLoop :distance Distance))
  (connect DefaultInputNode In FilterDataByType_0 In)
  (connect FilterDataByType_0 InsideFilter Loop_0 In)
  (output Out Loop_0))
