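; PCG subgraph, the core process loop of a spline extend transform, that lengthens open splines at their ends by a distance parameter and passes closed splines through.
; An element-count branch separates two-point segments from longer splines, with near-identical chains for each.
; Each chain copies control-point forward vectors into a direction attribute and assigns distance-based Magnitude to first and last points, found by index.
; Both attributes are copied onto the spline to drive Offset Spline.
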(graph Transform_Splines_Extend_CoreProcessLoop
  (params (distance :type Double :default 5e+01))
  (bind FilterDataByAttribute_0 (Filter%20Data%20By%20Attribute :FilterMode FilterByValue :FilterOperator Equal :TargetAttribute @data.$isClosed :Threshold (struct :AttributeTypes (struct :Type Boolean)))) ; OPEN SLPINES, if closed pass through
  (bind GetElementsCount_6 (Get%20Element%20Count))
  (bind AttributeCompareOp_1 (Attribute%20Compare%20Op))
  (bind Branch_1 (Branch)) ; Branch segment vs 2+ points splines
  ;[0.15 0.15 0.15 0.5 50] +2 POINTS SPLINES
  (bind GetSplineControlPoints_1 (Get%20Spline%20Control%20Points))
  (bind CopyAttributes_4 (Copy%20Attributes :InputSource $Rotation.Forward :OutputTarget direction))
  (bind CopyAttributes_0 (Copy%20Attributes :InputSource Direction))
  (bind AddAttribute_0 (Add%20Attribute :OutputTarget Magnitude))
  (bind AttributeFilter_3 (Filter%20Attribute%20Elements :bUseConstantThreshold true :Operator Equal :TargetAttribute $Index :ThresholdAttribute @Last))
  (bind GetElementsCount_1 (Get%20Element%20Count))
  (bind AttributeMathsOp_16 (Attribute%20Maths%20Op :Operation Subtract))
  (bind AttributeFilter_1 (Filter%20Attribute%20Elements :Operator Equal :TargetAttribute $Index :ThresholdAttribute @Last))
  (bind AttributeMathsOp_19 (Attribute%20Maths%20Op :Operation Multiply :InA distance))
  (bind AttributeMathsOp_17 (Attribute%20Maths%20Op :InputSource1 Magnitude :Operation Set))
  (bind AttributeMathsOp_18 (Attribute%20Maths%20Op :InputSource1 Magnitude :Operation Set :InB distance))
  (bind MergePoints_2 (Merge%20Points))
  (bind CopyAttributes_1 (Copy%20Attributes :InputSource Magnitude :OutputTarget Magnitude))
  (bind OffsetSpline_0 (Offset%20Spline :DirectionAttribute Direction :MagnitudeAttribute Magnitude))
  ;[0.15 0.15 0.15 0.5 50] 2 POINTS SPLINES
  (bind GetSplineControlPoints_3 (Get%20Spline%20Control%20Points))
  (bind CopyAttributes_6 (Copy%20Attributes :InputSource $Rotation.Forward :OutputTarget direction))
  (bind CopyAttributes_2 (Copy%20Attributes :InputSource Direction))
  (bind AddAttribute_1 (Add%20Attribute :OutputTarget Magnitude))
  (bind AttributeFilter_6 (Filter%20Attribute%20Elements :bUseConstantThreshold true :Operator Equal :TargetAttribute $Index :ThresholdAttribute @Last))
  (bind AttributeMathsOp_24 (Attribute%20Maths%20Op :Operation Multiply :InA distance))
  (bind AttributeMathsOp_22 (Attribute%20Maths%20Op :InputSource1 Magnitude :Operation Set))
  (bind GetElementsCount_4 (Get%20Element%20Count))
  (bind AttributeMathsOp_21 (Attribute%20Maths%20Op :Operation Subtract))
  (bind AttributeFilter_5 (Filter%20Attribute%20Elements :Operator Equal :TargetAttribute $Index :ThresholdAttribute @Last))
  (bind AttributeMathsOp_23 (Attribute%20Maths%20Op :InputSource1 Magnitude :Operation Set :InB distance))
  (bind MergePoints_3 (Merge%20Points))
  (bind CopyAttributes_3 (Copy%20Attributes :InputSource Magnitude :OutputTarget Magnitude))
  (bind OffsetSpline_1 (Offset%20Spline :DirectionAttribute Direction :MagnitudeAttribute Magnitude))
  (connect GetSplineControlPoints_1 Out AddAttribute_0 In)
  (connect GetSplineControlPoints_3 Out AddAttribute_1 In)
  (connect GetElementsCount_6 Out AttributeCompareOp_1 InA)
  (connect AttributeMathsOp_16 Out AttributeFilter_1 Filter)
  (connect AttributeFilter_3 OutsideFilter AttributeFilter_1 In)
  (connect AddAttribute_0 Out AttributeFilter_3 In)
  (connect AttributeMathsOp_21 Out AttributeFilter_5 Filter)
  (connect AttributeFilter_6 OutsideFilter AttributeFilter_5 In)
  (connect AddAttribute_1 Out AttributeFilter_6 In)
  (connect GetElementsCount_1 Out AttributeMathsOp_16 InA)
  (connect AttributeFilter_3 InsideFilter AttributeMathsOp_17 InA)
  (connect AttributeMathsOp_19 Out AttributeMathsOp_17 InB)
  (connect AttributeFilter_1 InsideFilter AttributeMathsOp_18 InA)
  (connect GetElementsCount_4 Out AttributeMathsOp_21 InA)
  (connect AttributeFilter_6 InsideFilter AttributeMathsOp_22 InA)
  (connect AttributeMathsOp_24 Out AttributeMathsOp_22 InB)
  (connect AttributeFilter_5 InsideFilter AttributeMathsOp_23 InA)
  (connect AttributeCompareOp_1 Out Branch_1 bOutputToB)
  (connect FilterDataByAttribute_0 InsideFilter Branch_1 In)
  (connect CopyAttributes_4 Out CopyAttributes_0 Source)
  (connect Branch_1 Output%20A CopyAttributes_0 Target)
  (connect MergePoints_2 Out CopyAttributes_1 Source)
  (connect CopyAttributes_0 Out CopyAttributes_1 Target)
  (connect CopyAttributes_6 Out CopyAttributes_2 Source)
  (connect Branch_1 Output%20B CopyAttributes_2 Target)
  (connect MergePoints_3 Out CopyAttributes_3 Source)
  (connect CopyAttributes_2 Out CopyAttributes_3 Target)
  (connect GetSplineControlPoints_1 Out CopyAttributes_4 Source)
  (connect GetSplineControlPoints_1 Out CopyAttributes_4 Target)
  (connect GetSplineControlPoints_3 Out CopyAttributes_6 Source)
  (connect GetSplineControlPoints_3 Out CopyAttributes_6 Target)
  (connect DefaultInputNode In FilterDataByAttribute_0 In)
  (connect AttributeFilter_3 OutsideFilter GetElementsCount_1 In)
  (connect AttributeFilter_6 OutsideFilter GetElementsCount_4 In)
  (connect FilterDataByAttribute_0 InsideFilter GetElementsCount_6 In)
  (connect Branch_1 Output%20A GetSplineControlPoints_1 In)
  (connect Branch_1 Output%20B GetSplineControlPoints_3 In)
  (connect AttributeFilter_1 OutsideFilter MergePoints_2 In)
  (connect AttributeMathsOp_17 Out MergePoints_2 In)
  (connect AttributeMathsOp_18 Out MergePoints_2 In)
  (connect AttributeMathsOp_22 Out MergePoints_3 In)
  (connect AttributeMathsOp_23 Out MergePoints_3 In)
  (connect CopyAttributes_1 Out OffsetSpline_0 In)
  (connect CopyAttributes_1 Out OffsetSpline_0 Magnitude)
  (connect CopyAttributes_3 Out OffsetSpline_1 In)
  (connect CopyAttributes_3 Out OffsetSpline_1 Magnitude)
  (output Out FilterDataByAttribute_0.OutsideFilter)
  (output Out OffsetSpline_0)
  (output Out OffsetSpline_1))
