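; PCG subgraph that serves as the core process of a spline reverse transform primitive.
; It keeps input data tagged "Input:0", the primary input, then narrows it to spline data.
; A Spline Direction node at its default settings processes those splines and feeds the output.
; Data rejected by either filter is not passed through.
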(graph Transform_Splines_Reverse_CoreProcess
  ;[0 0 0 0.5 12] Primary/Secondary Inputs
  (bind FilterDataByTag_0 (Filter%20Data%20By%20Tag :SelectedTags "Input:0"))
  ;[0 0 0 0.5] Core Process
  (bind FilterDataByType_0 (Filter%20Data%20By%20Type :TargetType (struct :Ids [(struct :Struct /Script/PCG.PCGDataTypeInfoSpline)])))
  (bind SplineDirection_0 (Spline%20Direction))
  (connect DefaultInputNode In FilterDataByTag_0 In)
  (connect FilterDataByTag_0 InsideFilter FilterDataByType_0 In)
  (connect FilterDataByType_0 InsideFilter SplineDirection_0 In)
  (output Out SplineDirection_0))
