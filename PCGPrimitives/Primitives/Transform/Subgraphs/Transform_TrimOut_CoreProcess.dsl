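; PCG subgraph, the core of a trim-out transform, that cuts a set length off the start of splines tagged Input:0.
; Nonzero lengths split each spline by distance; zero passes splines through unchanged.
; It keeps the last piece per OriginatingSplineIndex by copying indices, sorting descending, removing duplicates and filtering by index.
; A second, direction-changed trim chain exists, but its branch has no data input and its result never reaches the output.
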(graph Transform_TrimOut_CoreProcess
  (params (lenght :type Double))
  (bind FilterDataByTag_0 (Filter%20Data%20By%20Tag :SelectedTags "Input:0"))
  (bind AttributeCompareOp_3 (Attribute%20Compare%20Op :Operation NotEqual :InA lenght))
  (bind Branch_3 (Branch))
  (bind FilterDataByType_1 (Filter%20Data%20By%20Type :TargetType (struct :Ids [(struct :Struct /Script/PCG.PCGDataTypeInfoSpline)])))
  (bind SplitSpline_2 (Split%20Spline :Mode ByDistance :Constant-param lenght))
  (bind FilterDataByAttribute_2 (Filter%20Data%20By%20Attribute :Attribute @Data.OriginatingSplineIndex))
  (bind ExtractAttribute_2 (Extract%20Attribute%20at%20Index :bForceOutputAttributeToBeInElementsDomain true :InputSource @Data.OriginatingSplineIndex))
  (bind MergeAttributes_2 (Merge%20Attributes))
  (bind CopyAttributes_4 (Copy%20Attributes :InputSource $Index :OutputTarget IndexBck))
  (bind SortAttributes_1 (Sort%20Attributes :bUseStableSort false :SortAttributes [(struct :InputSource IndexBck :SortMethod Descending)]))
  (bind AttributeRemoveDuplicates_0 (Attribute%20Remove%20Duplicates :AttributeSelectors [OriginatingSplineIndex]))
  (bind SortAttributes_2 (Sort%20Attributes :bUseStableSort false :SortAttributes [(struct :InputSource IndexBck)]))
  (bind AttributeReduce_1 (Attribute%20Reduce :InputSource IndexBck :Operation Join))
  (bind FilterDataByIndex_0 (Filter%20Data%20By%20Index))
  (bind AttributeCompareOp_2 (Attribute%20Compare%20Op :Operation NotEqual))
  (bind Branch_2 (Branch))
  (bind SplineDirection_3 (Spline%20Direction))
  (bind SplitSpline_3 (Split%20Spline :Mode ByDistance))
  (bind FilterDataByAttribute_3 (Filter%20Data%20By%20Attribute :Attribute @Data.OriginatingSplineIndex))
  (bind ExtractAttribute_3 (Extract%20Attribute%20at%20Index :bForceOutputAttributeToBeInElementsDomain true :InputSource @Data.OriginatingSplineIndex))
  (bind MergeAttributes_3 (Merge%20Attributes))
  (bind CopyAttributes_5 (Copy%20Attributes :InputSource $Index :OutputTarget IndexBck))
  (bind SortAttributes_3 (Sort%20Attributes :bUseStableSort false :SortAttributes [(struct :InputSource IndexBck :SortMethod Descending)]))
  (bind AttributeRemoveDuplicates_1 (Attribute%20Remove%20Duplicates :AttributeSelectors [OriginatingSplineIndex]))
  (bind SortAttributes_4 (Sort%20Attributes :bUseStableSort false :SortAttributes [(struct :InputSource IndexBck)]))
  (bind AttributeReduce_2 (Attribute%20Reduce :InputSource IndexBck :Operation Join))
  (bind FilterDataByIndex_1 (Filter%20Data%20By%20Index))
  (bind DeleteAttributes_0 (Delete%20Attributes :MetadataDomain Data :SelectedAttributes "OriginatingSplineIndex"))
  (bind FilterDataByType_2 (Filter%20Data%20By%20Type :TargetType (struct :Ids [(struct :Struct /Script/PCG.PCGDataTypeInfoSpline)])))
  (connect SortAttributes_2 Out AttributeReduce_1 In)
  (connect SortAttributes_4 Out AttributeReduce_2 In)
  (connect SortAttributes_1 Out AttributeRemoveDuplicates_0 In)
  (connect SortAttributes_3 Out AttributeRemoveDuplicates_1 In)
  (connect AttributeCompareOp_2 Out Branch_2 bOutputToB)
  (connect AttributeCompareOp_3 Out Branch_3 bOutputToB)
  (connect FilterDataByTag_0 InsideFilter Branch_3 In)
  (connect MergeAttributes_2 Out CopyAttributes_4 Source)
  (connect MergeAttributes_2 Out CopyAttributes_4 Target)
  (connect MergeAttributes_3 Out CopyAttributes_5 Source)
  (connect MergeAttributes_3 Out CopyAttributes_5 Target)
  (connect FilterDataByIndex_1 InsideFilter DeleteAttributes_0 In)
  (connect FilterDataByAttribute_2 InsideFilter ExtractAttribute_2 In)
  (connect FilterDataByAttribute_3 InsideFilter ExtractAttribute_3 In)
  (connect SplitSpline_2 Out FilterDataByAttribute_2 In)
  (connect SplitSpline_3 Out FilterDataByAttribute_3 In)
  (connect FilterDataByAttribute_2 InsideFilter FilterDataByIndex_0 In)
  (connect AttributeReduce_1 Out FilterDataByIndex_0 SelectedIndices)
  (connect FilterDataByAttribute_3 InsideFilter FilterDataByIndex_1 In)
  (connect AttributeReduce_2 Out FilterDataByIndex_1 SelectedIndices)
  (connect DefaultInputNode In FilterDataByTag_0 In)
  (connect Branch_3 Output%20B FilterDataByType_1 In)
  (connect Branch_3 Output%20A FilterDataByType_2 In)
  (connect ExtractAttribute_2 Out MergeAttributes_2 In)
  (connect ExtractAttribute_3 Out MergeAttributes_3 In)
  (connect CopyAttributes_4 Out SortAttributes_1 In)
  (connect AttributeRemoveDuplicates_0 Out SortAttributes_2 In)
  (connect CopyAttributes_5 Out SortAttributes_3 In)
  (connect AttributeRemoveDuplicates_1 Out SortAttributes_4 In)
  (connect Branch_2 Output%20B SplineDirection_3 In)
  (connect FilterDataByType_1 InsideFilter SplitSpline_2 In)
  (connect SplineDirection_3 Out SplitSpline_3 In)
  (output Out Branch_2.Output%20A)
  (output Out FilterDataByIndex_0.InsideFilter)
  (output Out FilterDataByType_2.InsideFilter))
